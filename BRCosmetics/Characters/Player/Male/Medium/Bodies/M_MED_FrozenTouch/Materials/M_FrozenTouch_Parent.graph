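# Parent character material for the Frozen Touch player skin, adding an ice effect to the shared 2023 character material.
# The ice function sits between the shared PreFX and PostFX functions, taking PreFX output and passing modified material attributes onward.
# Named reroutes expose the Tech Art mask and ice mask.
# It uses material attributes with an expression-driven shading model and supports skeletal meshes, morph targets and mesh deformers.
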Material /BRCosmetics/Characters/Player/Male/Medium/Bodies/M_MED_FrozenTouch/Materials/M_FrozenTouch_Parent
graph {
 properties {
  bUseMaterialAttributes=true
  bUsedWithMeshDeformer=true
  bUsedWithMorphTargets=true
  bUsedWithSkeletalMesh=true
  pixelDepthOffsetMode=PDOM_Legacy
  shadingModel=MSM_FromMaterialExpression
  usedShadingModels=Default Lit | Subsurface | Clear Coat | Subsurface Profile | Cloth
 }
 nodes [
  n1 = MaterialFunctionCall() {materialFunction=&MF_CharacterParent_2023_PreFX}
  n2 = MaterialFunctionCall!(In=@n1.Result, InMask=@n1.Result, InFXMask=@n1.Result) {materialFunction=&/MF_FrozenTouch_Ice}
  MaterialFunctionCall(
   Material=NamedRerouteDeclaration(None=@n2.Result) {name=InMaterialAttributes, nodeColor=<rgba 1 0.3281 0 1>}
   M=@n1.M
   Pre Skinned Local Position=@n1.M
   Distance Blend=@n1.M
   Fresnel=@n1.M
  ) {materialFunction=&MF_CharacterParent_2023_PostFX}
  NamedRerouteDeclaration(None=@n1.Tech Art Mask) {name=InFXMask, nodeColor=<rgba 1 0 0.2344 1>}
  NamedRerouteDeclaration(None=@n2.IceMask) {name=IceMask, nodeColor=<rgba 0 1 0.6484 1>}
 ]
}
