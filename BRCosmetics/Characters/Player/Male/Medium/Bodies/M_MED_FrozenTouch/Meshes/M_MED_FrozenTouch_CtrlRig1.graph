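# Control Rig for the FrozenTouch male medium body cosmetic that applies corrective deformation, mainly keeping the scarf and collar fitted as the character moves.
# It runs the shared DeformRig function, then a sequence of constraints and bone offsets.
# Spherical pose readers on clavicles, head and upper arms weight the offsets.
# Backpack, skydive, front-end and targeting variables adjust them, and a swing-twist DeltaRotation function turns neck twist into a weight.
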ControlRigBlueprint /BRCosmetics/Characters/Player/Male/Medium/Bodies/M_MED_FrozenTouch/Meshes/M_MED_FrozenTouch_CtrlRig1
graphs [
 ConstructionGraph {nodes [RigUnit_PrepareForExecution = &RigUnit_PrepareForExecution()]}
 New Function_ContainedGraph {nodes [
   n1 = RigVMFunctionEntryNode()
   RigVMFunction_MathBoolToInteger = &RigVMFunction_MathBoolToInteger(Value=@n1.onlyTwist)
   RigVMFunction_MathQuaternionSwingTwist = &RigVMFunction_MathQuaternionSwingTwist(
    twistAxis=@n1.twistAxis
    Input=&RigUnit_GetRelativeTransformForItem(
     bChildInitial=False
     bParentInitial=False
     child=(Type=Bone,Name="None")
     parent=(Type=Bone,Name="None")
     child.Name=@n1.child
     parent.Name=@n1.parent
     parent.Type=@n1.parentType
    ).RelativeTransform.Rotation
   )
   RigVMFunction_MathQuaternionToEuler = &RigVMFunction_MathQuaternionToEuler(
    Value=RigVMDispatchNode(
     Values=((),())
     Index=@RigVMFunction_MathBoolToInteger.Result
     Values.0=@RigVMFunction_MathQuaternionSwingTwist.Swing
     Values.1=@RigVMFunction_MathQuaternionSwingTwist.Twist
    ) {template="DISPATCH_RigVMDispatch_SelectInt32(in Index,in Values,out Result)"}.Result
    RotationOrder=@n1.Rotation Order
   )
   > DISPATCH_RigVMDispatch_Print = RigVMDispatchNode(ScreenColor=(R=1,G=1,B=1,A=1), ScreenDuration=1.5, Enabled=@n1.debugRotation, Value=@RigVMFunction_MathQuaternionToEuler.Result) {template="DISPATCH_RigVMDispatch_Print(in Prefix,in Value,in Enabled,in ScreenDuration,in ScreenColor)"}
   Clamp = &RigVMFunction_MathIntClamp(Value=@n1.primAxis)
   DISPATCH_RigVMDispatch_SelectInt32 = RigVMDispatchNode(
    Values=(0,0,0)
    Index=@Clamp.Result
    Values.0=@RigVMFunction_MathQuaternionToEuler.Result.X
    Values.1=@RigVMFunction_MathQuaternionToEuler.Result.Y
    Values.2=@RigVMFunction_MathQuaternionToEuler.Result.Z
   ) {template="DISPATCH_RigVMDispatch_SelectInt32(in Index,in Values,out Result)"}
   Remap = &RigVMFunction_MathDoubleRemap(
    TargetMaximum=1
    TargetMinimum=0
    bClamp=False
    Value=@DISPATCH_RigVMDispatch_SelectInt32.Result
    SourceMaximum=@n1.max
    SourceMinimum=@n1.min
   )
   > RigVMFunctionReturnNode(outputWeight=&RigVMFunction_MathDoubleClamp(Value=@Remap.Result).Result)
  ]}
 RigVMFunctionLibrary {nodes [DeltaRotation_1 = RigVMCollapseNode(ExecuteContext=(), Rotation Order=ZYX, twistAxis=(X=1,Y=0,Z=0))]}
 RigVMFunction_Sequence_1_ContainedGraph {nodes [
   n1 = RigVMFunctionEntryNode()
   RigVMFunction_Sequence = &RigVMFunction_Sequence(ExecuteContext=@n1.ExecuteContext)
   n2 = &RigVMFunction_Sequence(ExecuteContext=@RigVMFunction_Sequence.B)
   n3 = &RigVMFunction_Sequence(ExecuteContext=@n2.B)
   n4 = &RigVMFunction_Sequence(ExecuteContext=@n3.B)
   n5 = &RigVMFunction_Sequence(ExecuteContext=@n4.B)
   n6 = &RigVMFunction_Sequence(ExecuteContext=@n5.B)
   RigVMFunctionReturnNode(A=@RigVMFunction_Sequence.A, B=@n2.A, C=@n3.A, D=@n4.A, E=@n5.A, F=@n6.A, G=@n6.B)
  ]}
 RigVMModel {nodes [
   RigUnit_BeginExecution = &RigUnit_BeginExecution()
   FN_DeformRig_FNC = RigVMFunctionReferenceNode(
    DeformRig_ Curve="(deformRig_l_upperarm_up_alpha_72_#1=False,deformRig_r_upperarm_up_alpha_73_#2=False,deformRig_l_upperarm_fr_alpha_74_#3=False,deformRig_r_upperarm_fr_alpha_75_CB51B8 ...(truncated)"
    DeformRig_Bool="(upperarm_1_#4=True,elbow_3_#5=True,wrist_7_#6=True,groin_8_#7=True,glute_17_#8=True,knee_9_40C408324CA261 ...(truncated)"
    DeformRig_Debug=(upperarm_debug_18_#4=False,wrist_debug_20_#6=False,lat_debug_26_#9=False,groinGlute_debug_27_#7=False)
    DeformRig_Value="(upperarm_up_18_#10=(Rotation=(X=0,Y=0,Z=0,W=1),Translation=(X=0,Y=0,Z=5),Scale3D=(X=0.7,Y=1,Z=1)),upperarm_fwd_fr_19_#11=(Rotation=(X=0,Y=0,Z=0,W=1),Translation=(X=0,Y=10,Z=0),Scale3D=(X=0.5,Y=1,Z= ...(truncated)"
    ExecuteContext=@RigUnit_BeginExecution.ExecutePin
   )
   > RigVMFunction_Sequence = RigVMAggregateNode()
   &RigUnit_ParentConstraint(
    AdvancedSettings=(InterpolationType=Average,RotationOrderForFilter=XZY)
    Child=(Type=Bone,Name="base_scarfCollar_frontMain_01")
    Parents=((Item=(Type=Bone,Name="clavicle_r"),Weight=0.25),(Item=(Type=Bone,Name="clavicle_l"),Weight=0.25),(Item=(Type=Bone,Name="spine_05"),Weight=0.5))
    Weight=1
    ExecutePin=@RigVMFunction_Sequence.A
   )
   &RigUnit_ModifyTransforms(
    ItemToModify=((Item=(Type=Bone,Name="clavicle_l"),Transform=(Rotation=(X=0,Y=0,Z=0,W=1),Translation=(X=0,Y=0,Z=0),Scale3D=(X=1,Y=1,Z=1))))
    Weight=1
    WeightMinimum=0
    ExecutePin=@RigVMFunction_Sequence.B
   )
   > n1 = &RigUnit_SphericalPoseReader(
    ActiveRegionScaleFactors=(PositiveWidth=0.3,NegativeWidth=3,PositiveHeight=1.5,NegativeHeight=1.5)
    ActiveRegionSize=0.1
    Debug=(bDrawDebug=false,bDraw2D=False,bDrawLocalAxes=False,DebugScale=25,DebugSegments=20,DebugThickness=0.25)
    DriverAxis=(X=1,Y=0,Z=0)
    DriverItem=(Type=Bone,Name="clavicle_l")
    FalloffRegionScaleFactors=(PositiveWidth=0.8,NegativeWidth=0,PositiveHeight=1,NegativeHeight=1)
    FalloffSize=0.2
    FlipHeightScaling=False
    FlipWidthScaling=False
    OptionalParentItem=(Type=Bone,Name="spine_05")
    RotationOffset=(X=0,Y=-60,Z=0)
   )
   > &RigUnit_ModifyTransforms(
    ItemToModify="((Item=(Type=Bone,Name=\"base_scarfCollar_frontMain_01_l\"),Transform=(Rotation=(X=0,Y=0,Z=0,W=-1),Translation=(X=-2.5,Y=0,Z=1.3),Scale3D=(X=1,Y=1,Z=1))),(Item=(Type=Bone,Name=\"base_scarfCollar_backMain_01_l\"),Transform=(Rotation=(X=0,Y=0,Z=0,W=-1),Translati ...(truncated)"
    WeightMinimum=0
    Weight=@n1.OutputParam
   )
   > n2 = &RigUnit_SphericalPoseReader(
    ActiveRegionScaleFactors=(PositiveWidth=0.3,NegativeWidth=3,PositiveHeight=1.5,NegativeHeight=1.5)
    ActiveRegionSize=0.1
    Debug=(bDrawDebug=false,bDraw2D=False,bDrawLocalAxes=False,DebugScale=25,DebugSegments=20,DebugThickness=0.25)
    DriverAxis=(X=1,Y=0,Z=0)
    DriverItem=(Type=Bone,Name="clavicle_l")
    FalloffRegionScaleFactors=(PositiveWidth=0.8,NegativeWidth=0,PositiveHeight=1,NegativeHeight=1)
    FalloffSize=0.2
    FlipHeightScaling=False
    FlipWidthScaling=False
    OptionalParentItem=(Type=Bone,Name="spine_05")
    RotationOffset=(X=0,Y=-35,Z=0)
   )
   > &RigUnit_ModifyTransforms(
    ItemToModify="((Item=(Type=Bone,Name=\"base_scarfCollar_frontMain_01_l\"),Transform=(Rotation=(X=0,Y=-0.1219,Z=0,W=0.9925),Translation=(X=-2,Y=-4,Z=-2.3),Scale3D=(X=1,Y=1,Z=1))),(Item=(Type=Bone,Name=\"base_scarfCollar_backMain_01_l\"),Transform=(Rotation=(X=0.1662,Y=-0.159 ...(truncated)"
    WeightMinimum=0
    Weight=@n2.OutputParam
   )
   &RigUnit_ModifyTransforms(
    ItemToModify=((Item=(Type=Bone,Name="clavicle_r"),Transform=(Rotation=(X=0,Y=0,Z=0,W=1),Translation=(X=0,Y=0,Z=0),Scale3D=(X=1,Y=1,Z=1))))
    Weight=1
    WeightMinimum=0
    ExecutePin=@RigVMFunction_Sequence.C
   )
   > n3 = &RigUnit_SphericalPoseReader(
    ActiveRegionScaleFactors=(PositiveWidth=0.3,NegativeWidth=3,PositiveHeight=1.5,NegativeHeight=1.5)
    ActiveRegionSize=0.1
    Debug=(bDrawDebug=false,bDraw2D=False,bDrawLocalAxes=False,DebugScale=25,DebugSegments=20,DebugThickness=0.25)
    DriverAxis=(X=-1,Y=0,Z=0)
    DriverItem=(Type=Bone,Name="clavicle_r")
    FalloffRegionScaleFactors=(PositiveWidth=0.8,NegativeWidth=0,PositiveHeight=1,NegativeHeight=1)
    FalloffSize=0.2
    FlipHeightScaling=False
    FlipWidthScaling=False
    OptionalParentItem=(Type=Bone,Name="spine_05")
    RotationOffset=(X=0,Y=120,Z=0)
   )
   > &RigUnit_ModifyTransforms(
    ItemToModify="((Item=(Type=Bone,Name=\"base_scarfCollar_frontMain_01_r\"),Transform=(Rotation=(X=0,Y=0,Z=0,W=-1),Translation=(X=-3,Y=0,Z=0),Scale3D=(X=1,Y=1,Z=1))),(Item=(Type=Bone,Name=\"base_scarfCollar_backMain_01_r\"),Transform=(Rotation=(X=0,Y=0,Z=0,W=-1),Translation=( ...(truncated)"
    WeightMinimum=0
    Weight=@n3.OutputParam
   )
   > n4 = &RigUnit_SphericalPoseReader(
    ActiveRegionScaleFactors=(PositiveWidth=0.3,NegativeWidth=3,PositiveHeight=1.5,NegativeHeight=1.5)
    ActiveRegionSize=0.1
    Debug=(bDrawDebug=false,bDraw2D=False,bDrawLocalAxes=False,DebugScale=25,DebugSegments=20,DebugThickness=0.25)
    DriverAxis=(X=-1,Y=0,Z=0)
    DriverItem=(Type=Bone,Name="clavicle_r")
    FalloffRegionScaleFactors=(PositiveWidth=0.8,NegativeWidth=0,PositiveHeight=1,NegativeHeight=1)
    FalloffSize=0.2
    FlipHeightScaling=False
    FlipWidthScaling=False
    OptionalParentItem=(Type=Bone,Name="spine_05")
    RotationOffset=(X=0,Y=145,Z=0)
   )
   > &RigUnit_ModifyTransforms(
    ItemToModify="((Item=(Type=Bone,Name=\"base_scarfCollar_frontMain_01_r\"),Transform=(Rotation=(X=0,Y=0,Z=0,W=1),Translation=(X=-1,Y=0,Z=-2),Scale3D=(X=1,Y=1,Z=1))),(Item=(Type=Bone,Name=\"base_scarfCollar_backMain_01_r\"),Transform=(Rotation=(X=0,Y=0,Z=0,W=1),Translation=(X ...(truncated)"
    WeightMinimum=0
    Weight=@n4.OutputParam
   )
   &RigUnit_ParentConstraint(
    AdvancedSettings=(InterpolationType=Average,RotationOrderForFilter=XZY)
    Child=(Type=Bone,Name="base_BackScarf_01_r")
    Parents=((Item=(Type=Bone,Name="clavicle_r"),Weight=1))
    Weight=1
    Parents.0.Weight=RigVMVariableNode() {variable=is_wear_backpack_float}.Value
    ExecutePin=@RigVMFunction_Sequence.F
   )
   AlphaInterp = &RigVMFunction_AlphaInterp(InterpSpeedDecreasing=5, InterpSpeedIncreasing=5, Scale=1, bClampResult=False, Value=RigVMVariableNode() {variable=SkyDiveFloat}.Value)
   > &RigUnit_ModifyTransforms(
    ItemToModify=((Item=(Type=Bone,Name="base_BackScarf_01_r"),Transform=(Rotation=(X=0,Y=0.0872,Z=0,W=0.9962),Translation=(X=0,Y=0,Z=0),Scale3D=(X=1,Y=1,Z=1))))
    WeightMinimum=0
    Weight=@AlphaInterp.Result
   )
   > &RigUnit_ModifyTransforms(
    ItemToModify="((Item=(Type=Bone,Name=\"base_BackScarf_01_r\"),Transform=(Rotation=(X=-0.0015,Y=0.0174,Z=0.0871,W=0.996),Translation=(X=-1,Y=0,Z=0.5),Scale3D=(X=1,Y=1,Z=1))),(Item=(Type=Bone,Name=\"dyn_BackScarf_02_r\"),Transform=(Rotation=(X=-0.0868,Y=0.0868,Z=0.0076,W=0.99 ...(truncated)"
    WeightMinimum=0
    Weight=RigVMVariableNode() {variable=isInFrontEnd}.Value
   )
   > DeltaRotation_1_1_1_1_2 = RigVMFunctionReferenceNode(
    Rotation Order=ZYX
    child=neck_01
    debugRotation=false
    max=25
    min=14
    onlyTwist=true
    parent=head_null_sp
    parentType=Null
    primAxis=0
    twistAxis=(X=1,Y=0,Z=0)
   )
   > &RigUnit_ModifyTransforms(
    ItemToModify=((Item=(Type=Bone,Name="dyn_BackScarf_01_r"),Transform=(Rotation=(X=0,Y=0,Z=0,W=1),Translation=(X=0,Y=0,Z=-1),Scale3D=(X=1,Y=1,Z=1))))
    WeightMinimum=0
    Weight=&RigVMFunction_MathDoubleMul(A=@DeltaRotation_1_1_1_1_2.outputWeight, B=RigVMVariableNode() {variable=isInFrontEnd}.Value).Result
   )
   n5 = &RigUnit_SphericalPoseReader(
    ActiveRegionScaleFactors=(PositiveWidth=1,NegativeWidth=1,PositiveHeight=1,NegativeHeight=1)
    ActiveRegionSize=0.15
    Debug=(bDrawDebug=false,bDraw2D=False,bDrawLocalAxes=true,DebugScale=25,DebugSegments=20,DebugThickness=0.25)
    DriverAxis=(X=1,Y=0,Z=0)
    DriverItem=(Type=Bone,Name="head")
    FalloffRegionScaleFactors=(PositiveWidth=1,NegativeWidth=1,PositiveHeight=1,NegativeHeight=1)
    FalloffSize=0.35
    FlipHeightScaling=False
    FlipWidthScaling=False
    OptionalParentItem=(Type=Bone,Name="spine_05")
    RotationOffset=(X=-80,Y=0,Z=0)
    ExecutePin=@RigVMFunction_Sequence.E
   )
   > &RigUnit_ModifyTransforms(
    ItemToModify="((Item=(Type=Bone,Name=\"dyn_scarfCollar_back_01_r\"),Transform=(Rotation=(X=0,Y=0,Z=0,W=1),Translation=(X=8,Y=0,Z=0),Scale3D=(X=1,Y=1,Z=1))),(Item=(Type=Bone,Name=\"dyn_scarfCollar_back_01\"),Transform=(Rotation=(X=0,Y=0,Z=0,W=1),Translation=(X=6,Y=0,Z=0),Sca ...(truncated)"
    WeightMinimum=0
    Weight=@n5.OutputParam
   )
   > &RigUnit_ModifyTransforms(
    ItemToModify="((Item=(Type=Bone,Name=\"base_scarfCollar_back_lower_01_r\"),Transform=(Rotation=(X=0,Y=0,Z=0,W=1),Translation=(X=-5,Y=0,Z=0),Scale3D=(X=1,Y=1,Z=1))),(Item=(Type=Bone,Name=\"base_scarfCollar_back_Lower_01\"),Transform=(Rotation=(X=0,Y=0,Z=0,W=1),Translation=(X ...(truncated)"
    WeightMinimum=0
    Weight=@n5.OutputParam
   )
   > &RigUnit_ModifyTransforms(
    ItemToModify=((Item=(Type=Bone,Name="base_BackScarf_01_r"),Transform=(Rotation=(X=0,Y=0,Z=0,W=1),Translation=(X=2,Y=0,Z=0),Scale3D=(X=1,Y=1,Z=1))))
    WeightMinimum=0
    Weight=@n5.OutputParam
   )
   > n6 = &RigUnit_SphericalPoseReader(
    ActiveRegionScaleFactors=(PositiveWidth=1,NegativeWidth=1,PositiveHeight=1,NegativeHeight=1)
    ActiveRegionSize=0.13
    Debug=(bDrawDebug=false,bDraw2D=False,bDrawLocalAxes=true,DebugScale=25,DebugSegments=20,DebugThickness=0.25)
    DriverAxis=(X=1,Y=0,Z=0)
    DriverItem=(Type=Bone,Name="head")
    FalloffRegionScaleFactors=(PositiveWidth=1,NegativeWidth=1,PositiveHeight=1,NegativeHeight=1)
    FalloffSize=0.28
    FlipHeightScaling=False
    FlipWidthScaling=False
    OptionalParentItem=(Type=Bone,Name="spine_05")
    RotationOffset=(X=-50,Y=0,Z=40)
   )
   Multiply = &RigVMFunction_MathDoubleMul(A=@n6.OutputParam, B=RigVMVariableNode() {variable=IsTargeting}.Value)
   > &RigUnit_ModifyTransforms(
    ItemToModify="((Item=(Type=Bone,Name=\"base_scarfCollar_backMain_01_r\"),Transform=(Rotation=(X=0,Y=0,Z=0,W=1),Translation=(X=5,Y=0,Z=0),Scale3D=(X=1,Y=1,Z=1))),(Item=(Type=Bone,Name=\"base_scarfCollar_backMain_01\"),Transform=(Rotation=(X=0,Y=0,Z=0,W=1),Translation=(X=3,Y= ...(truncated)"
    WeightMinimum=0
    Weight=@Multiply.Result
   )
   > &RigUnit_ModifyTransforms(
    ItemToModify=((Item=(Type=Bone,Name="head"),Transform=(Rotation=(X=0,Y=0,Z=0,W=1),Translation=(X=0,Y=0,Z=0),Scale3D=(X=1,Y=1,Z=1))))
    Weight=1
    WeightMinimum=0
   )
   > n7 = &RigUnit_SphericalPoseReader(
    ActiveRegionScaleFactors=(PositiveWidth=1,NegativeWidth=1,PositiveHeight=1,NegativeHeight=1)
    ActiveRegionSize=0.11
    Debug=(bDrawDebug=false,bDraw2D=False,bDrawLocalAxes=true,DebugScale=25,DebugSegments=20,DebugThickness=0.25)
    DriverAxis=(X=0,Y=1,Z=0)
    DriverItem=(Type=Bone,Name="head")
    FalloffRegionScaleFactors=(PositiveWidth=1,NegativeWidth=1,PositiveHeight=1,NegativeHeight=1)
    FalloffSize=0.23
    FlipHeightScaling=False
    FlipWidthScaling=False
    OptionalParentItem=(Type=Bone,Name="spine_05")
    RotationOffset=(X=0,Y=180,Z=0)
   )
   > &RigUnit_ModifyTransforms(
    ItemToModify="((Item=(Type=Bone,Name=\"base_scarfCollar_frontMain_01_l\"),Transform=(Rotation=(X=0,Y=0,Z=0,W=-1),Translation=(X=0,Y=0,Z=1),Scale3D=(X=1,Y=1,Z=1))),(Item=(Type=Bone,Name=\"base_scarfCollar_backMain_01_l\"),Transform=(Rotation=(X=0,Y=0,Z=0,W=1),Translation=(X= ...(truncated)"
    WeightMinimum=0
    Weight=@n7.OutputParam
   )
   > &RigUnit_ModifyTransforms(
    ItemToModify="((Item=(Type=Bone,Name=\"base_scarfCollar_frontMain_01\"),Transform=(Rotation=(X=0,Y=0,Z=0,W=1),Translation=(X=0,Y=-1,Z=0),Scale3D=(X=1,Y=1,Z=1))),(Item=(Type=Bone,Name=\"base_FrontScarf_01_l\"),Transform=(Rotation=(X=0,Y=0,Z=0,W=-1),Translation=(X=0,Y=-1,Z=0) ...(truncated)"
    WeightMinimum=0
    Weight=@n7.OutputParam
   )
   > n8 = &RigUnit_SphericalPoseReader(
    ActiveRegionScaleFactors=(PositiveWidth=1,NegativeWidth=1,PositiveHeight=1,NegativeHeight=1)
    ActiveRegionSize=0.05
    Debug=(bDrawDebug=false,bDraw2D=False,bDrawLocalAxes=true,DebugScale=25,DebugSegments=20,DebugThickness=0.25)
    DriverAxis=(X=0,Y=1,Z=0)
    DriverItem=(Type=Bone,Name="head")
    FalloffRegionScaleFactors=(PositiveWidth=1,NegativeWidth=1,PositiveHeight=1,NegativeHeight=1)
    FalloffSize=0.18
    FlipHeightScaling=False
    FlipWidthScaling=False
    OptionalParentItem=(Type=Bone,Name="neck_01")
    RotationOffset=(X=30,Y=0,Z=0)
   )
   > &RigUnit_ModifyTransforms(
    ItemToModify="((Item=(Type=Bone,Name=\"base_scarfCollar_frontMain_01_r\"),Transform=(Rotation=(X=0,Y=0,Z=0,W=-1),Translation=(X=0,Y=0,Z=2),Scale3D=(X=1,Y=1,Z=1))),(Item=(Type=Bone,Name=\"base_scarfCollar_backMain_01_r\"),Transform=(Rotation=(X=-0.0872,Y=0,Z=0,W=0.9962),Tran ...(truncated)"
    WeightMinimum=0
    Weight=@n8.OutputParam
   )
   > n9 = &RigUnit_SphericalPoseReader(
    ActiveRegionScaleFactors=(PositiveWidth=1,NegativeWidth=1,PositiveHeight=1,NegativeHeight=1)
    ActiveRegionSize=0.12
    Debug=(bDrawDebug=false,bDraw2D=False,bDrawLocalAxes=true,DebugScale=25,DebugSegments=20,DebugThickness=0.25)
    DriverAxis=(X=-1,Y=0,Z=0)
    DriverItem=(Type=Bone,Name="upperarm_r")
    FalloffRegionScaleFactors=(PositiveWidth=1,NegativeWidth=1,PositiveHeight=1,NegativeHeight=1)
    FalloffSize=0.3
    FlipHeightScaling=False
    FlipWidthScaling=False
    OptionalParentItem=(Type=Bone,Name="spine_05")
    RotationOffset=(X=-100,Y=0,Z=40)
   )
   > &RigUnit_ModifyTransforms(
    ItemToModify=((Item=(Type=Bone,Name="base_scarfCollar_frontMain_01_r"),Transform=(Rotation=(X=0,Y=0,Z=0,W=1),Translation=(X=0,Y=6,Z=-2),Scale3D=(X=1,Y=1,Z=1))))
    WeightMinimum=0
    Weight=@n9.OutputParam
   )
   &RigUnit_ModifyTransforms(
    ItemToModify="((Item=(Type=Bone,Name=\"base_FrontScarf_03_l\"),Transform=(Rotation=(X=0.3651,Y=0.0811,Z=-0.0319,W=0.9269),Translation=(X=0,Y=0,Z=0),Scale3D=(X=1,Y=1,Z=1))),(Item=(Type=Bone,Name=\"base_scarfCollar_backMain_01_r\"),Transform=(Rotation=(X=0,Y=-0.0872,Z=0,W=0.9 ...(truncated)"
    WeightMinimum=0
    Weight=RigVMVariableNode() {variable=isInFrontEnd}.Value
    ExecutePin=@RigVMFunction_Sequence.D
   )
   > n10 = &RigUnit_SphericalPoseReader(
    ActiveRegionScaleFactors=(PositiveWidth=1,NegativeWidth=1,PositiveHeight=1,NegativeHeight=1)
    ActiveRegionSize=0.12
    Debug=(bDrawDebug=false,bDraw2D=False,bDrawLocalAxes=false,DebugScale=25,DebugSegments=20,DebugThickness=0.25)
    DriverAxis=(X=1,Y=0,Z=0)
    DriverItem=(Type=Bone,Name="upperarm_l")
    FalloffRegionScaleFactors=(PositiveWidth=1,NegativeWidth=1,PositiveHeight=1,NegativeHeight=1)
    FalloffSize=0.35
    FlipHeightScaling=False
    FlipWidthScaling=False
    OptionalParentItem=(Type=Bone,Name="spine_05")
    RotationOffset=(X=120,Y=0,Z=-10)
   )
   > &RigUnit_ModifyTransforms(
    ItemToModify="((Item=(Type=Bone,Name=\"None\"),Transform=(Rotation=(X=0,Y=0,Z=0,W=1),Translation=(X=0,Y=0,Z=0),Scale3D=(X=1,Y=1,Z=1))),(Item=(Type=Bone,Name=\"base_FrontScarf_01_l\"),Transform=(Rotation=(X=0,Y=0,Z=0.0872,W=0.9962),Translation=(X=-1,Y=1,Z=1),Scale3D=(X=1,Y=1 ...(truncated)"
    WeightMinimum=0
    Weight=@n10.OutputParam
   )
   n11 = &RigUnit_SphericalPoseReader(
    ActiveRegionScaleFactors=(PositiveWidth=1,NegativeWidth=1,PositiveHeight=1,NegativeHeight=1)
    ActiveRegionSize=0.2
    Debug=(bDrawDebug=false,bDraw2D=False,bDrawLocalAxes=true,DebugScale=25,DebugSegments=20,DebugThickness=0.25)
    DriverAxis=(X=1,Y=0,Z=0)
    DriverItem=(Type=Bone,Name="upperarm_l")
    ExecutePin=()
    FalloffRegionScaleFactors=(PositiveWidth=1,NegativeWidth=1,PositiveHeight=1,NegativeHeight=1)
    FalloffSize=0.3
    FlipHeightScaling=False
    FlipWidthScaling=False
    OptionalParentItem=(Type=Bone,Name="spine_05")
    RotationOffset=(X=0,Y=30,Z=0)
   )
   > &RigUnit_ModifyTransforms(
    ItemToModify="((Item=(Type=Bone,Name=\"None\"),Transform=(Rotation=(X=0,Y=0,Z=0,W=1),Translation=(X=0,Y=0,Z=0),Scale3D=(X=1,Y=1,Z=1))),(Item=(Type=Bone,Name=\"base_FrontScarf_01_l\"),Transform=(Rotation=(X=0,Y=0.0872,Z=0,W=0.9962),Translation=(X=0,Y=0,Z=5),Scale3D=(X=1,Y=1, ...(truncated)"
    WeightMinimum=0
    Weight=@n11.OutputParam
   )
  ]}
]
hierarchy [
 root {loc=<0 0 0>, rot=<0 0 0 1>}
 pelvis {parent=root, loc=<0 -0.7071 93.1191>, rot=<0.0127 0.707 -0.0127 -0.707>}
 spine_01 {parent=pelvis, loc=<6.7494 -0.6388 0>, rot=<0 0 0.034 -0.9994>}
 spine_02 {parent=spine_01, loc=<7.2518 -0.035 0>, rot=<0 0 0.0238 -0.9997>}
 spine_03 {parent=spine_02, loc=<6.0104 0 0>, rot=<0 0 0.0149 -0.9999>}
 spine_04 {parent=spine_03, loc=<8.1024 0 0>, rot=<0 0 -0.0132 -0.9999>}
 spine_05 {parent=spine_04, loc=<10.0146 0 0>, rot=<0 0 -0.0015 -1>}
 neck_01 {parent=spine_05, loc=<11.3138 0.1529 0>, rot=<0 0 -0.2003 -0.9797>}
 neck_02 {parent=neck_01, loc=<4.5627 0 0>, rot=<0 0 0.0504 -0.9987>}
 head {parent=neck_02, loc=<5.3914 0 0>, rot=<0 0 0.115 -0.9934>}
 dyn_hairA_01 {parent=head, loc=<10.6588 8.7607 -1.3164>, rot=<0.5065 0.546 0.6633 -0.0729>}
 dyn_hairA_02 {parent=dyn_hairA_01, loc=<3.453 0 0>, rot=<-0.0157 -0.3678 -0.0397 -0.9289>}
 dyn_hairA_03 {parent=dyn_hairA_02, loc=<3.6529 0 0>, rot=<-0.0418 -0.1375 -0.1807 -0.973>}
 dyn_hairA_04 {parent=dyn_hairA_03, loc=<3.2415 0 0>, rot=<-0.0172 -0.0472 0.3413 -0.9386>}
 dyn_hairA_05 {parent=dyn_hairA_04, loc=<1.6069 0 0>, rot=<0 0 0 1>}
 dyn_hairB_01 {parent=head, loc=<10.879 3.2848 7.5609>, rot=<0.3306 0.0346 0.4293 -0.8397>}
 dyn_hairB_02 {parent=dyn_hairB_01, loc=<1.255 0 0>, rot=<0 0 0 1>}
 dyn_hairC_01 {parent=head, loc=<9.6397 1.9458 7.5175>, rot=<0.329 0.2009 0.4709 -0.7935>}
 dyn_hairC_02 {parent=dyn_hairC_01, loc=<1.8976 0 0>, rot=<0.0017 -0.0149 -0.1148 -0.9933>}
 dyn_hairC_03 {parent=dyn_hairC_02, loc=<2.0936 0 0>, rot=<0 0 0 1>}
 dyn_hairD_01 {parent=head, loc=<8.3833 1.3056 7.3262>, rot=<0.1823 0.0511 0.5865 -0.7875>}
 dyn_hairD_02 {parent=dyn_hairD_01, loc=<1.288 0 0>, rot=<0 0 0 1>}
 dyn_hairE_01 {parent=head, loc=<6.3723 -2.5927 6.1371>, rot=<-0.0415 -0.0237 0.9411 -0.3348>}
 dyn_hairE_02 {parent=dyn_hairE_01, loc=<1.5498 0 0>, rot=<0 0 0 1>}
 dyn_hairF_01 {parent=head, loc=<4.169 -5.5852 2.0333>, rot=<-0.2075 -0.76 0.6145 0.0421>}
 dyn_hairF_02 {parent=dyn_hairF_01, loc=<1.6046 0 0>, rot=<0 0 0 1>}
 dyn_hairG_01 {parent=head, loc=<6.9024 -5.7943 3.9664>, rot=<0.0439 -0.5704 0.6315 -0.5234>}
 dyn_hairG_02 {parent=dyn_hairG_01, loc=<2.0664 0 0>, rot=<0 0 0 1>}
 dyn_hairH_01 {parent=head, loc=<7.8155 -6.4145 3.2547>, rot=<-0.109 -0.3526 0.8603 -0.3518>}
 dyn_hairH_02 {parent=dyn_hairH_01, loc=<0.8177 0 0>, rot=<0 0 0 1>}
 dyn_hairI_01 {parent=head, loc=<9.5112 -5.6125 4.6812>, rot=<0.0015 -0.2154 0.7387 -0.6387>}
 dyn_hairI_02 {parent=dyn_hairI_01, loc=<1.21 0 0>, rot=<0 0 0 1>}
 dyn_hairJ_01 {parent=head, loc=<10.4435 -1.4689 7.2398>, rot=<0.2693 -0.0199 0.5748 -0.7725>}
 dyn_hairJ_02 {parent=dyn_hairJ_01, loc=<2.6333 0 0>, rot=<0 0 0 1>}
 dyn_hairK_01 {parent=head, loc=<11.5777 -1.7389 6.6012>, rot=<0.1257 -0.1496 0.6248 -0.7559>}
 dyn_hairK_02 {parent=dyn_hairK_01, loc=<1.2009 0 0>, rot=<0.3555 -0.4265 -0.5325 -0.6388>}
 dyn_hairL_01 {parent=head, loc=<12.0419 3.8378 6.1684>, rot=<0.3564 0.2446 0.5204 -0.7364>}
 dyn_hairL_02 {parent=dyn_hairL_01, loc=<2.3423 0 0>, rot=<-0.0884 -0.346 -0.4112 -0.8387>}
 dyn_hairL_03 {parent=dyn_hairL_02, loc=<1.7891 0 0>, rot=<0 0 0 1>}
 dyn_hairM_01 {parent=head, loc=<13.8129 -0.7908 5.9703>, rot=<0.0489 -0.1143 0.3827 -0.9155>}
 dyn_hairM_02 {parent=dyn_hairM_01, loc=<2.2989 0 0>, rot=<0 0 0 1>}
 dyn_hairN_01 {parent=head, loc=<14.7402 0.308 3.5585>, rot=<0.7784 -0.1832 0.4274 -0.4217>}
 dyn_hairN_02 {parent=dyn_hairN_01, loc=<2.2512 0 0>, rot=<0 0 0 1>}
 dyn_hairO_01 {parent=head, loc=<14.6613 -2.1979 0.9638>, rot=<0.6076 -0.3693 0.3787 -0.5925>}
 dyn_hairO_02 {parent=dyn_hairO_01, loc=<1.3192 0 0>, rot=<0 0 0 1>}
 dyn_hairP_01 {parent=head, loc=<15.4494 0.7991 -1.4822>, rot=<0.7798 -0.3236 -0.035 -0.5348>}
 dyn_hairP_02 {parent=dyn_hairP_01, loc=<1.7691 0 0>, rot=<0 0 0 1>}
 dyn_hairQ_01 {parent=head, loc=<13.8607 3.0534 -5.1009>, rot=<0.8427 -0.2311 0.0712 -0.4811>}
 dyn_hairQ_02 {parent=dyn_hairQ_01, loc=<1.9441 0 0>, rot=<0 0 0 1>}
 dyn_hairR_01 {parent=head, loc=<13.6529 -1.6173 -4.5707>, rot=<0.8976 -0.1889 0.3801 -0.1193>}
 dyn_hairR_02 {parent=dyn_hairR_01, loc=<1.7542 0 0>, rot=<0 0 0 1>}
 dyn_hairS_01 {parent=head, loc=<7.1212 -0.5906 -6.5356>, rot=<0.3961 -0.901 -0.0698 0.1625>}
 dyn_hairS_02 {parent=dyn_hairS_01, loc=<1.3204 0 0>, rot=<-0.3273 0.753 -0.2276 -0.5236>}
 dyn_hairT_01 {parent=head, loc=<9.2905 -5.7275 -4.2606>, rot=<0.5643 -0.7721 0.2665 0.1199>}
 dyn_hairT_02 {parent=dyn_hairT_01, loc=<1.5519 0 0>, rot=<0 0 0 1>}
 dyn_hairU_01 {parent=head, loc=<9.5644 -7.3343 -2.3719>, rot=<0.5553 -0.7604 0.3343 0.0409>}
 dyn_hairU_02 {parent=dyn_hairU_01, loc=<1.4993 0 0>, rot=<0.0156 0.0641 0.236 -0.9695>}
 dyn_hairU_03 {parent=dyn_hairU_02, loc=<1.8304 0 0>, rot=<0 0 0 1>}
 dyn_hairV_01 {parent=head, loc=<10.7717 -6.8527 0.63>, rot=<0.0851 -0.3789 0.529 -0.7545>}
 dyn_hairV_02 {parent=dyn_hairV_01, loc=<2.9408 0 0>, rot=<0 0 0 1>}
 dyn_hairW_01 {parent=head, loc=<7.3956 -7.7923 0.1312>, rot=<0.0386 -0.6782 0.7334 -0.0279>}
 dyn_hairW_02 {parent=dyn_hairW_01, loc=<2.1595 0 0>, rot=<0 0 0 1>}
 dyn_hairX_01 {parent=head, loc=<4.3 -5.3577 -2.0553>, rot=<0.1564 -0.6514 0.6968 0.2565>}
 dyn_hairX_02 {parent=dyn_hairX_01, loc=<0.9656 0 0>, rot=<0 0 0 1>}
 dyn_hairY_01 {parent=head, loc=<4.155 -6.2819 -1.3498>, rot=<-0.185 -0.674 0.6916 -0.182>}
 dyn_hairY_02 {parent=dyn_hairY_01, loc=<1.5733 0 0>, rot=<0 0 0 1>}
 dyn_hairZ_01 {parent=head, loc=<2.6576 -5.151 0.3804>, rot=<-0.1554 -0.6451 0.7468 0.0445>}
 dyn_hairZ_02 {parent=dyn_hairZ_01, loc=<1.36 0 0>, rot=<0 0 0 1>}
 FACIAL_C_FacialRoot {parent=head, loc=<2.1524 -0.746 -0.0002>, rot=<0.4879 -0.5118 -0.4879 -0.5118>}
 FACIAL_C_Forehead {parent=FACIAL_C_FacialRoot, loc=<-0.0007 -2.3874 9.4198>, rot=<-0.0431 0 0 -0.9991>}
 FACIAL_L_ForeheadIn {parent=FACIAL_C_FacialRoot, loc=<1.6059 -2.3664 9.3065>, rot=<-0.0603 -0.0696 0.0043 -0.9957>}
 FACIAL_R_ForeheadIn {parent=FACIAL_C_FacialRoot, loc=<-1.6074 -2.3662 9.3058>, rot=<-0.0603 0.0696 -0.0043 -0.9957>}
 FACIAL_L_ForeheadMid {parent=FACIAL_C_FacialRoot, loc=<2.8932 -2.6155 8.8328>, rot=<-0.0706 -0.1757 -0.0206 -0.9817>}
 FACIAL_R_ForeheadMid {parent=FACIAL_C_FacialRoot, loc=<-2.8947 -2.6156 8.8325>, rot=<-0.0706 0.1756 0.0205 -0.9817>}
 FACIAL_L_ForeheadOut {parent=FACIAL_C_FacialRoot, loc=<4.0037 -2.5908 8.0671>, rot=<-0.121 -0.2556 -0.0791 -0.9559>}
 FACIAL_R_ForeheadOut {parent=FACIAL_C_FacialRoot, loc=<-4.0049 -2.5913 8.0662>, rot=<-0.121 0.2556 0.079 -0.9559>}
 FACIAL_L_EyesackUpper {parent=FACIAL_C_FacialRoot, loc=<2.9822 -1.5756 9.2744>, rot=<0.0492 -0.0843 -0.0645 -0.9931>}
 FACIAL_R_EyesackUpper {parent=FACIAL_C_FacialRoot, loc=<-2.984 -1.5738 9.2782>, rot=<0.0492 0.0843 0.0645 -0.9931>}
 FACIAL_L_EyelidUpperFurrow {parent=FACIAL_C_FacialRoot, loc=<2.8083 -1.413 9.0687>, rot=<-0.0297 -0.1509 -0.0796 -0.9849>}
 FACIAL_R_EyelidUpperFurrow {parent=FACIAL_C_FacialRoot, loc=<-2.809 -1.4173 9.0752>, rot=<-0.0297 0.1508 0.0796 -0.9849>}
 FACIAL_L_EyelidUpperA {parent=FACIAL_C_FacialRoot, loc=<2.7455 -0.9785 7.3619>, rot=<-0.1475 0 0 -0.9891>}
 FACIAL_R_EyelidUpperA {parent=FACIAL_C_FacialRoot, loc=<-2.7383 -0.9786 7.3669>, rot=<-0.1475 0 0 -0.9891>}
 FACIAL_L_Eye {parent=FACIAL_C_FacialRoot, loc=<2.7382 -0.9861 7.3324>, rot=<0.0004 0 0 -1>}
 FACIAL_R_Eye {parent=FACIAL_C_FacialRoot, loc=<-2.7364 -0.9832 7.336>, rot=<0.0003 0 0 -1>}
 FACIAL_L_EyelidLowerA {parent=FACIAL_C_FacialRoot, loc=<2.7455 -0.9785 7.3619>, rot=<0.2844 0 0 -0.9587>}
 FACIAL_R_EyelidLowerA {parent=FACIAL_C_FacialRoot, loc=<-2.7383 -0.9786 7.3669>, rot=<0.2843 0 0 -0.9587>}
 FACIAL_L_EyeCornerInner {parent=FACIAL_C_FacialRoot, loc=<1.5386 -0.5012 8.4897>, rot=<0.0004 0 0 -1>}
 FACIAL_R_EyeCornerInner {parent=FACIAL_C_FacialRoot, loc=<-1.5353 -0.5011 8.4864>, rot=<0.0003 0 0 -1>}
 FACIAL_L_EyeCornerOuter {parent=FACIAL_C_FacialRoot, loc=<4.1836 -1.0829 7.9716>, rot=<0.0003 -0.3826 0.0001 -0.9239>}
 FACIAL_R_EyeCornerOuter {parent=FACIAL_C_FacialRoot, loc=<-4.1892 -1.0816 7.9658>, rot=<0.0003 0.3826 -0.0002 -0.9239>}
 FACIAL_L_EyesackLower {parent=FACIAL_C_FacialRoot, loc=<2.981 0.1389 8.1772>, rot=<-0.0219 -0.1397 -0.0222 -0.9897>}
 FACIAL_R_EyesackLower {parent=FACIAL_C_FacialRoot, loc=<-2.9814 0.1524 8.1748>, rot=<-0.0219 0.1397 0.0222 -0.9897>}
 FACIAL_L_CheekInner {parent=FACIAL_C_FacialRoot, loc=<1.8943 1.0654 8.2523>, rot=<-0.1229 -0.0881 -0.1971 -0.9687>}
 FACIAL_R_CheekInner {parent=FACIAL_C_FacialRoot, loc=<-1.8947 1.0662 8.2512>, rot=<-0.1229 0.0881 0.197 -0.9687>}
 FACIAL_L_CheekOuter {parent=FACIAL_C_FacialRoot, loc=<3.4311 1.2104 6.824>, rot=<0.0555 -0.3769 0.1005 -0.9191>}
 FACIAL_R_CheekOuter {parent=FACIAL_C_FacialRoot, loc=<-3.4316 1.2095 6.8245>, rot=<0.0555 0.3769 -0.1005 -0.9191>}
 FACIAL_L_NasolabialBulge {parent=FACIAL_C_FacialRoot, loc=<3.0477 4.4774 6.6353>, rot=<0.0583 -0.3773 0.0239 -0.9239>}
 FACIAL_R_NasolabialBulge {parent=FACIAL_C_FacialRoot, loc=<-3.0483 4.4755 6.6363>, rot=<0.0582 0.3773 -0.024 -0.9239>}
 FACIAL_L_NasolabialFurrow {parent=FACIAL_C_FacialRoot, loc=<2.5783 3.161 8.7258>, rot=<-0.0329 -0.126 -0.3636 -0.9224>}
 FACIAL_R_NasolabialFurrow {parent=FACIAL_C_FacialRoot, loc=<-2.5789 3.1637 8.7256>, rot=<-0.033 0.1259 0.3636 -0.9224>}
 FACIAL_L_CheekLower {parent=FACIAL_C_FacialRoot, loc=<3.4641 4.2495 6.017>, rot=<0.0349 -0.4269 0.0579 -0.9018>}
 FACIAL_R_CheekLower {parent=FACIAL_C_FacialRoot, loc=<-3.4661 4.2469 6.0179>, rot=<0.0349 0.4268 -0.0579 -0.9018>}
 FACIAL_C_Nose {parent=FACIAL_C_FacialRoot, loc=<-0.001 0.6701 9.6516>, rot=<-0.13 0 0 -0.9915>}
 FACIAL_C_NoseLower {parent=FACIAL_C_Nose, loc=<0.0008 2.7363 -0.3656>, rot=<0.1305 0 0 -0.9914>}
 FACIAL_L_Nostril {parent=FACIAL_C_Nose, loc=<0.9783 2.0662 -0.4941>, rot=<0.0173 0.1305 0.0023 -0.9913>}
 FACIAL_R_Nostril {parent=FACIAL_C_Nose, loc=<-0.9772 2.0649 -0.4984>, rot=<0.0173 -0.1305 -0.0023 -0.9913>}
 FACIAL_C_NoseTip {parent=FACIAL_C_Nose, loc=<0.0004 2.0646 0.4739>, rot=<0.0872 0 0 -0.9962>}
 FACIAL_C_MouthUpper {parent=FACIAL_C_FacialRoot, loc=<0.0001 4.697 7.5263>, rot=<0.0004 0 0 -1>}
 FACIAL_C_LipUpper {parent=FACIAL_C_MouthUpper, loc=<-0.0013 -0.1726 2.2324>, rot=<-0.0872 0 0 -0.9962>}
 FACIAL_L_LipUpper {parent=FACIAL_C_MouthUpper, loc=<0.9872 -0.0762 1.8969>, rot=<-0.0778 -0.1601 -0.0379 -0.9833>}
 FACIAL_R_LipUpper {parent=FACIAL_C_MouthUpper, loc=<-0.9884 -0.0773 1.8992>, rot=<-0.0778 0.1601 0.0379 -0.9833>}
 FACIAL_L_LipUpperOuter {parent=FACIAL_C_MouthUpper, loc=<1.6445 0.1593 1.4064>, rot=<-0.0984 -0.2293 -0.0983 -0.9634>}
 FACIAL_R_LipUpperOuter {parent=FACIAL_C_MouthUpper, loc=<-1.6428 0.1609 1.4085>, rot=<-0.0984 0.2293 0.0983 -0.9634>}
 FACIAL_L_LipCorner {parent=FACIAL_C_MouthUpper, loc=<2.1404 0.5475 0.7824>, rot=<0 -0.2078 0 -0.9782>}
 FACIAL_R_LipCorner {parent=FACIAL_C_MouthUpper, loc=<-2.1459 0.5485 0.7922>, rot=<0 0.2078 0 -0.9782>}
 FACIAL_L_Masseter {parent=FACIAL_C_FacialRoot, loc=<5.4457 1.7977 4.7409>, rot=<0.0697 0.1129 0.0078 -0.9911>}
 FACIAL_R_Masseter {parent=FACIAL_C_FacialRoot, loc=<-5.446 1.8012 4.7377>, rot=<0.0697 -0.1129 -0.0079 -0.9911>}
 FACIAL_C_UnderChin {parent=FACIAL_C_FacialRoot, loc=<-0.0006 8.5128 5.0968>, rot=<0.5002 0 0 -0.8659>}
 FACIAL_L_UnderChin {parent=FACIAL_C_FacialRoot, loc=<2.7448 7.0994 4.0275>, rot=<0.4029 -0.3548 0.0824 -0.8397>}
 FACIAL_R_UnderChin {parent=FACIAL_C_FacialRoot, loc=<-2.7461 7.1003 4.0267>, rot=<0.4028 0.3548 -0.0825 -0.8397>}
 FACIAL_C_TeethUpper {parent=FACIAL_C_FacialRoot, loc=<0.0016 5.2278 9.0991>, rot=<0.0004 0 0 -1>}
 FACIAL_C_LowerLipRotation {parent=FACIAL_C_FacialRoot, loc=<-0.0034 2.1601 2.7036>, rot=<0.3424 0 0 -0.9396>}
 FACIAL_C_MouthLower {parent=FACIAL_C_LowerLipRotation, loc=<0.0035 -1.1609 5.3241>, rot=<-0.342 0 0 -0.9397>}
 FACIAL_C_LipLower {parent=FACIAL_C_MouthLower, loc=<0.0028 1.2701 1.9157>, rot=<0.0872 0 0 -0.9962>}
 FACIAL_L_LipLower {parent=FACIAL_C_MouthLower, loc=<0.814 1.1687 1.699>, rot=<0.0927 -0.1294 0.0272 -0.9869>}
 FACIAL_R_LipLower {parent=FACIAL_C_MouthLower, loc=<-0.809 1.1689 1.7011>, rot=<0.0927 0.1294 -0.0272 -0.9869>}
 FACIAL_L_LipLowerOuter {parent=FACIAL_C_MouthLower, loc=<1.5002 0.9204 1.2627>, rot=<0.0913 -0.2321 0.0318 -0.9679>}
 FACIAL_R_LipLowerOuter {parent=FACIAL_C_MouthLower, loc=<-1.4967 0.9197 1.2649>, rot=<0.0913 0.2321 -0.0318 -0.9679>}
 FACIAL_C_TeethLower {parent=FACIAL_C_LowerLipRotation, loc=<0.0048 -1.7067 6.7821>, rot=<-0.342 0 0 -0.9397>}
 FACIAL_C_Tongue1 {parent=FACIAL_C_TeethLower, loc=<-0.004 -0.0797 -5.0762>, rot=<-0.2588 0 0 -0.9659>}
 FACIAL_C_Tongue2 {parent=FACIAL_C_Tongue1, loc=<0.0011 0.431 1.4544>, rot=<0.2588 0 0 -0.9659>}
 FACIAL_C_Tongue3 {parent=FACIAL_C_Tongue2, loc=<0.0009 0.2308 1.3201>, rot=<0.1305 0 0 -0.9914>}
 FACIAL_C_Tongue4 {parent=FACIAL_C_Tongue3, loc=<0.0015 -0.1486 1.1173>, rot=<0.0436 0 0 -0.999>}
 FACIAL_C_TongueUpper3 {parent=FACIAL_C_Tongue3, loc=<0.0002 -0.3554 -0.1315>, rot=<0 0 0 1>}
 FACIAL_C_TongueLower3 {parent=FACIAL_C_Tongue3, loc=<0.0002 0.7722 0.0863>, rot=<0 0 0 1>}
 FACIAL_L_TongueSide3 {parent=FACIAL_C_Tongue3, loc=<1.4828 0.0526 0.1604>, rot=<0 0 0 1>}
 FACIAL_R_TongueSide3 {parent=FACIAL_C_Tongue3, loc=<-1.4825 0.0517 0.1633>, rot=<0 0 0 1>}
 FACIAL_C_TongueUpper2 {parent=FACIAL_C_Tongue2, loc=<0.0002 -0.5361 -0.0018>, rot=<0 0 0 1>}
 FACIAL_L_TongueSide2 {parent=FACIAL_C_Tongue2, loc=<1.7194 0.1492 0.3667>, rot=<0 0 0 1>}
 FACIAL_R_TongueSide2 {parent=FACIAL_C_Tongue2, loc=<-1.7191 0.1479 0.3681>, rot=<0 0 0 1>}
 FACIAL_C_Jaw {parent=FACIAL_C_FacialRoot, loc=<-0.0034 2.1601 2.7036>, rot=<0.3424 0 0 -0.9396>}
 FACIAL_C_Jawline {parent=FACIAL_C_Jaw, loc=<0.0036 2.1929 7.6405>, rot=<0.2586 0 0 -0.966>}
 FACIAL_L_ChinSide {parent=FACIAL_C_Jaw, loc=<2.5413 0.3594 6.8846>, rot=<-0.2208 -0.3738 0.0499 -0.8995>}
 FACIAL_R_ChinSide {parent=FACIAL_C_Jaw, loc=<-2.5307 0.3595 6.8841>, rot=<-0.2208 0.3738 -0.0499 -0.8995>}
 FACIAL_C_Chin {parent=FACIAL_C_Jaw, loc=<0.0043 0.8199 7.6562>, rot=<-0.2591 0 0 -0.9659>}
 FACIAL_L_Ear {parent=FACIAL_C_FacialRoot, loc=<6.1023 -0.0143 1.8408>, rot=<-0.0259 -0.0007 -0.0262 -0.9993>}
 FACIAL_R_Ear {parent=FACIAL_C_FacialRoot, loc=<-6.103 -0.0123 1.844>, rot=<-0.0259 0.0007 0.0261 -0.9993>}
 FACIAL_L_Jawline {parent=FACIAL_C_FacialRoot, loc=<3.7581 5.3818 5.2422>, rot=<0.2097 0.1577 -0.0548 -0.9634>}
 FACIAL_R_Jawline {parent=FACIAL_C_FacialRoot, loc=<-3.7557 5.3791 5.2395>, rot=<0.2097 -0.1577 0.0547 -0.9634>}
 FACIAL_C_NoseBridge {parent=FACIAL_C_FacialRoot, loc=<-0.0001 -1.1168 10.1291>, rot=<0.0442 0 0 -0.999>}
 FACIAL_L_EyelidUpperB {parent=FACIAL_C_FacialRoot, loc=<2.7455 -0.9785 7.3619>, rot=<-0.2246 0 0 -0.9745>}
 FACIAL_L_EyelidLowerB {parent=FACIAL_C_FacialRoot, loc=<2.7455 -0.9785 7.3619>, rot=<0.3176 0 0 -0.9482>}
 FACIAL_R_EyelidLowerB {parent=FACIAL_C_FacialRoot, loc=<-2.7383 -0.9786 7.3669>, rot=<0.3259 0 0 -0.9454>}
 FACIAL_R_EyelidUpperB {parent=FACIAL_C_FacialRoot, loc=<-2.7383 -0.9786 7.3669>, rot=<-0.2246 0 0 -0.9744>}
 attach_cape {parent=spine_05, loc=<9.017 -11.714 0>, rot=<0.0616 -0.7044 -0.0616 -0.7044>}
 cape {parent=attach_cape, type=Null, loc=<0 -1.8 0>, rot=<0 0 0 1>}
 attach_backpack {parent=spine_05, loc=<-0.123 -13.473 0>, rot=<-0.0093 -0.707 0.0093 -0.707>}
 Backpack_BR {parent=attach_backpack, type=Null, loc=<0 0 0>, rot=<0 0 0 1>}
 clavicle_l {parent=spine_05, loc=<4.3575 -0.1466 -2.3223>, rot=<0.0465 -0.7289 -0.009 -0.6829>}
 upperarm_l {parent=clavicle_l, loc=<15.5111 0 0>, rot=<0.0226 -0.3234 -0.0088 -0.9459>}
 lowerarm_l {parent=upperarm_l, loc=<26.6514 0 0>, rot=<0.0391 -0.0099 -0.2453 -0.9686>}
 hand_l {parent=lowerarm_l, loc=<22.3491 0 0>, rot=<-0.7081 -0.0507 0.0975 -0.6975>}
 weapon_l {parent=hand_l, loc=<-0.1974 -3.4111 0.3811>, rot=<0 0 0 1>}
 thumb_01_l {parent=hand_l, loc=<1.3118 -0.3191 -2.459>, rot=<0.6546 -0.4042 0.0664 -0.6354>}
 thumb_02_l {parent=thumb_01_l, loc=<5.9336 -0.0448 0.0401>, rot=<-0.0037 0.0213 0.298 -0.9543>}
 thumb_03_l {parent=thumb_02_l, loc=<2.2223 0.3682 -0.0907>, rot=<0.0118 -0.013 0.0066 -0.9998>}
 index_metacarpal_l {parent=hand_l, loc=<1.6535 0.0183 -2.96>, rot=<0.0158 -0.0124 -0.0173 -0.9996>}
 index_01_l {parent=index_metacarpal_l, loc=<8.3008 0.0357 -0.0524>, rot=<-0.0044 0.0485 0.0924 -0.9945>}
 index_02_l {parent=index_01_l, loc=<3.3587 -0.3004 0.0452>, rot=<-0.002 0.0027 0.3225 -0.9466>}
 index_03_l {parent=index_02_l, loc=<2.5578 -0.123 -0.0252>, rot=<-0.002 0.0067 0.2558 -0.9667>}
 middle_metacarpal_l {parent=hand_l, loc=<2.1856 -0.1725 -0.5984>, rot=<0.006 0.0168 -0.0329 -0.9993>}
 middle_01_l {parent=middle_metacarpal_l, loc=<7.3275 -0.2447 0.0129>, rot=<-0.0208 0.0042 0.0858 -0.9961>}
 middle_02_l {parent=middle_01_l, loc=<4.0123 0.1401 0.1175>, rot=<-0.0218 0.0258 0.3512 -0.9357>}
 middle_03_l {parent=middle_02_l, loc=<2.6077 0.0534 0.1373>, rot=<0.0082 0.0188 0.1708 -0.9851>}
 ring_metacarpal_l {parent=hand_l, loc=<2.2074 -0.3625 1.3638>, rot=<-0.0284 0.0722 -0.0458 -0.9959>}
 ring_01_l {parent=ring_metacarpal_l, loc=<7.0295 -0.6197 -0.1446>, rot=<-0.0489 -0.0252 0.1892 -0.9804>}
 ring_02_l {parent=ring_01_l, loc=<3.6174 0.4647 -0.1027>, rot=<0 0 0.2806 -0.9598>}
 ring_03_l {parent=ring_02_l, loc=<2.4152 0.114 0.0199>, rot=<0 0 0.1189 -0.9929>}
 pinky_metacarpal_l {parent=hand_l, loc=<1.7616 -0.9753 3.3814>, rot=<-0.0397 0.0814 -0.069 -0.9935>}
 pinky_01_l {parent=pinky_metacarpal_l, loc=<7.1709 -0.4831 -0.1094>, rot=<-0.0654 -0.0035 0.2242 -0.9723>}
 pinky_02_l {parent=pinky_01_l, loc=<2.8274 0.0893 -0.1043>, rot=<0.0037 0.0011 0.2854 -0.9584>}
 pinky_03_l {parent=pinky_02_l, loc=<2.2046 0.0307 -0.0078>, rot=<-0.0002 0.0098 0.1445 -0.9895>}
 hand_attach_l {parent=hand_l, loc=<8.5 -3 0>, rot=<0.7071 0.7071 0 0>}
 deform_wrist_dn_base_l {parent=hand_l, loc=<0 -2.2342 0>, rot=<0 0 0 -1>}
 deform_wrist_dn_l {parent=deform_wrist_dn_base_l, loc=<0 0 0>, rot=<0 0 0 1>}
 deform_wrist_up_base_l {parent=hand_l, loc=<0 2.2682 0>, rot=<0 0 0 -1>}
 deform_wrist_up_l {parent=deform_wrist_up_base_l, loc=<0 0 0>, rot=<0 0 0 1>}
 lowerarm_twist_02_l {parent=lowerarm_l, loc=<9.6199 0 0>, rot=<0 0 0 -1>}
 lowerarm_twist_01_l {parent=lowerarm_l, loc=<17.8549 0 0>, rot=<0 0 0 -1>}
 deform_elbow_in_base_l {parent=lowerarm_l, loc=<0.2112 5.259 0.8274>, rot=<0 0 0 -1>}
 deform_elbow_in_l {parent=deform_elbow_in_base_l, loc=<0 0 0>, rot=<0 0 0 1>}
 deform_elbow_out_base_l {parent=lowerarm_l, loc=<0 -4 0>, rot=<0 0 0 -1>}
 deform_elbow_out_l {parent=deform_elbow_out_base_l, loc=<0 0 0>, rot=<0 0 0 1>}
 upperarm_twist_01_l {parent=upperarm_l, loc=<0.0041 0 0>, rot=<0 0 0 -1>}
 deform_upperarm_bk_base_l {parent=upperarm_twist_01_l, loc=<1.0813 -7.2193 -1.9453>, rot=<0 0 0 1>}
 deform_upperarm_bk_l {parent=deform_upperarm_bk_base_l, loc=<0 0 0>, rot=<0 0 0 1>}
 deform_upperarm_fr_base_l {parent=upperarm_twist_01_l, loc=<-0.3691 7.0506 -2.765>, rot=<0 0 0 1>}
 deform_upperarm_fr_l {parent=deform_upperarm_fr_base_l, loc=<0 0 0>, rot=<0 0 0 1>}
 deform_upperarm_up_base_l {parent=upperarm_twist_01_l, loc=<-3.2216 -1.1391 3.3311>, rot=<0 0 0 1>}
 deform_upperarm_up_l {parent=deform_upperarm_up_base_l, loc=<0 0 0>, rot=<0 0 0 1>}
 upperarm_twist_02_l {parent=upperarm_l, loc=<13.7941 0 0>, rot=<0 0 0 -1>}
 clavicle_r {parent=spine_05, loc=<4.3575 -0.1466 2.3223>, rot=<0.7289 0.0465 -0.6829 0.009>}
 upperarm_r {parent=clavicle_r, loc=<-15.5111 0 0>, rot=<0.0226 -0.3234 -0.0088 -0.9459>}
 lowerarm_r {parent=upperarm_r, loc=<-26.6514 0 0>, rot=<0.0391 -0.0099 -0.2453 -0.9686>}
 hand_r {parent=lowerarm_r, loc=<-22.3491 0 0>, rot=<-0.7081 -0.0507 0.0975 -0.6975>}
 weapon_r {parent=hand_r, loc=<0.1974 3.4111 -0.3811>, rot=<0 0 0 1>}
 thumb_01_r {parent=hand_r, loc=<-1.3118 0.3191 2.459>, rot=<0.6546 -0.4042 0.0664 -0.6354>}
 thumb_02_r {parent=thumb_01_r, loc=<-5.9336 0.0448 -0.0401>, rot=<-0.0037 0.0213 0.298 -0.9543>}
 thumb_03_r {parent=thumb_02_r, loc=<-2.2223 -0.3682 0.0907>, rot=<0.0118 -0.013 0.0066 -0.9998>}
 index_metacarpal_r {parent=hand_r, loc=<-1.6535 -0.0183 2.96>, rot=<0.0158 -0.0124 -0.0173 -0.9996>}
 index_01_r {parent=index_metacarpal_r, loc=<-8.3008 -0.0357 0.0524>, rot=<-0.0044 0.0485 0.0924 -0.9945>}
 index_02_r {parent=index_01_r, loc=<-3.3587 0.3004 -0.0452>, rot=<-0.002 0.0027 0.3225 -0.9466>}
 index_03_r {parent=index_02_r, loc=<-2.5578 0.123 0.0252>, rot=<-0.002 0.0067 0.2558 -0.9667>}
 middle_metacarpal_r {parent=hand_r, loc=<-2.1856 0.1725 0.5984>, rot=<0.006 0.0168 -0.0329 -0.9993>}
 middle_01_r {parent=middle_metacarpal_r, loc=<-7.3275 0.2447 -0.0129>, rot=<-0.0208 0.0042 0.0858 -0.9961>}
 middle_02_r {parent=middle_01_r, loc=<-4.0123 -0.1401 -0.1175>, rot=<-0.0218 0.0258 0.3512 -0.9357>}
 middle_03_r {parent=middle_02_r, loc=<-2.6077 -0.0534 -0.1373>, rot=<0.0082 0.0188 0.1708 -0.9851>}
 ring_metacarpal_r {parent=hand_r, loc=<-2.2074 0.3625 -1.3638>, rot=<-0.0284 0.0722 -0.0458 -0.9959>}
 ring_01_r {parent=ring_metacarpal_r, loc=<-7.0295 0.6197 0.1446>, rot=<-0.0489 -0.0252 0.1892 -0.9804>}
 ring_02_r {parent=ring_01_r, loc=<-3.6174 -0.4647 0.1027>, rot=<0 0 0.2806 -0.9598>}
 ring_03_r {parent=ring_02_r, loc=<-2.4152 -0.114 -0.0199>, rot=<0 0 0.1189 -0.9929>}
 pinky_metacarpal_r {parent=hand_r, loc=<-1.7616 0.9753 -3.3814>, rot=<-0.0397 0.0814 -0.069 -0.9935>}
 pinky_01_r {parent=pinky_metacarpal_r, loc=<-7.1709 0.4831 0.1094>, rot=<-0.0654 -0.0035 0.2242 -0.9723>}
 pinky_02_r {parent=pinky_01_r, loc=<-2.8274 -0.0893 0.1043>, rot=<0.0037 0.0011 0.2854 -0.9584>}
 pinky_03_r {parent=pinky_02_r, loc=<-2.2046 -0.0307 0.0078>, rot=<-0.0002 0.0098 0.1445 -0.9895>}
 hand_attach_r {parent=hand_r, loc=<-8.5001 2.9999 0>, rot=<0 0 -0.7071 -0.7071>}
 deform_wrist_dn_base_r {parent=hand_r, loc=<-0.0004 2.2345 -0.0001>, rot=<0 0 0 -1>}
 deform_wrist_dn_r {parent=deform_wrist_dn_base_r, loc=<0 0 0>, rot=<0 0 0 1>}
 deform_wrist_up_base_r {parent=hand_r, loc=<0.0003 -2.2684 0.0001>, rot=<0 0 0 -1>}
 deform_wrist_up_r {parent=deform_wrist_up_base_r, loc=<0 0 0>, rot=<0 0 0 1>}
 lowerarm_twist_02_r {parent=lowerarm_r, loc=<-9.6199 0 0>, rot=<0 0 0 -1>}
 lowerarm_twist_01_r {parent=lowerarm_r, loc=<-17.8549 0 0>, rot=<0 0 0 -1>}
 deform_elbow_in_base_r {parent=lowerarm_r, loc=<-0.2111 -5.2591 -0.8275>, rot=<0 0 0 -1>}
 deform_elbow_in_r {parent=deform_elbow_in_base_r, loc=<-0.0001 0 0.0002>, rot=<0 0 0 1>}
 deform_elbow_out_base_r {parent=lowerarm_r, loc=<0 4 -0.0001>, rot=<0 0 0 -1>}
 deform_elbow_out_r {parent=deform_elbow_out_base_r, loc=<0 0 0>, rot=<0 0 0 1>}
 upperarm_twist_01_r {parent=upperarm_r, loc=<-0.0041 0 0>, rot=<0 0 0 -1>}
 deform_upperarm_bk_base_r {parent=upperarm_twist_01_r, loc=<-1.0814 7.2192 1.9455>, rot=<0 0 0 1>}
 deform_upperarm_bk_r {parent=deform_upperarm_bk_base_r, loc=<0 0 0>, rot=<0 0 0 1>}
 deform_upperarm_fr_base_r {parent=upperarm_twist_01_r, loc=<0.3689 -7.0506 2.7652>, rot=<0 0 0 1>}
 deform_upperarm_fr_r {parent=deform_upperarm_fr_base_r, loc=<0 0 0>, rot=<0 0 0 1>}
 deform_upperarm_up_base_r {parent=upperarm_twist_01_r, loc=<3.2214 1.1391 -3.3308>, rot=<0 0 0 1>}
 deform_upperarm_up_r {parent=deform_upperarm_up_base_r, loc=<0 0 0>, rot=<0 0 0 -1>}
 upperarm_twist_02_r {parent=upperarm_r, loc=<-13.7941 0 0>, rot=<0 0 0 -1>}
 base_FrontScarf_01_l {parent=spine_05, loc=<4.1828 9.0627 -8.1417>, rot=<-0.0147 -0.7126 -0.6448 -0.2762>}
 dyn_FrontScarf_01_l {parent=base_FrontScarf_01_l, loc=<0 0 0>, rot=<0 0 0 1>}
 base_FrontScarf_02_l {parent=dyn_FrontScarf_01_l, loc=<5 0 0>, rot=<0 -0.0597 0 -0.9982>}
 dyn_FrontScarf_02_l {parent=base_FrontScarf_02_l, loc=<0 0 0>, rot=<0 0 0 1>}
 base_FrontScarf_03_l {parent=dyn_FrontScarf_02_l, loc=<5 0 0>, rot=<0 -0.0076 0 -1>}
 dyn_FrontScarf_03_l {parent=base_FrontScarf_03_l, loc=<0 0 0>, rot=<0 0 0 1>}
 dyn_FrontScarf_04_l {parent=dyn_FrontScarf_03_l, loc=<5 0 0>, rot=<0 0 0 1>}
 base_scarfCollar_frontMain_01 {parent=spine_05, loc=<4.6396 12.4372 -1.5821>, rot=<0.0669 -0.761 -0.6403 0.0798>}
 base_scarfCollar_front_lower_01 {parent=base_scarfCollar_frontMain_01, loc=<3.4208 0 0>, rot=<0 0 1 0>}
 dyn_scarfCollar_front_lower_01 {parent=base_scarfCollar_front_lower_01, loc=<0 0 0>, rot=<0 0 0 1>}
 dyn_scarfCollar_front_lower_02 {parent=dyn_scarfCollar_front_lower_01, loc=<3.5182 0 0>, rot=<0 0 0 1>}
 base_scarfCollar_front_01 {parent=base_scarfCollar_frontMain_01, loc=<-3.3432 0 0>, rot=<0 0 0 1>}
 dyn_scarfCollar_front_01 {parent=base_scarfCollar_front_01, loc=<0 0 0>, rot=<0 0 0 1>}
 dyn_scarfCollar_front_02 {parent=dyn_scarfCollar_front_01, loc=<3.5182 0 0>, rot=<0 0 0 1>}
 base_scarfCollar_frontMain_01_r {parent=spine_05, loc=<11.6548 7.5292 9.9977>, rot=<0.216 0.0566 0.9553 -0.1939>}
 base_scarfCollar_front_01_r {parent=base_scarfCollar_frontMain_01_r, loc=<-4.4607 0 0>, rot=<0 0 0 1>}
 dyn_scarfCollar_front_01_r {parent=base_scarfCollar_front_01_r, loc=<0 0 0>, rot=<0 0 0 1>}
 dyn_scarfCollar_front_02_r {parent=dyn_scarfCollar_front_01_r, loc=<3.5182 0 0>, rot=<0 0 0 1>}
 base_scarfCollar_front_lower_01_r {parent=base_scarfCollar_frontMain_01_r, loc=<4.5663 0 0>, rot=<0 0 1 0>}
 dun_scarfCollar_front_lower_01_r {parent=base_scarfCollar_front_lower_01_r, loc=<0 0 0>, rot=<0 0 0 1>}
 dun_scarfCollar_front_lower_02_r {parent=dun_scarfCollar_front_lower_01_r, loc=<3.5182 0 0>, rot=<0 0 0 1>}
 base_scarfCollar_frontMain_01_l {parent=spine_05, loc=<12.2602 4.2709 -12.3492>, rot=<0.2724 -0.9186 -0.015 -0.286>}
 base_scarfCollar_front_Lower_01_l {parent=base_scarfCollar_frontMain_01_l, loc=<3.7836 0 0>, rot=<0 0 1 0>}
 dyn_scarfCollar_front_Lower_01_l {parent=base_scarfCollar_front_Lower_01_l, loc=<0 0 0>, rot=<0 0 0 1>}
 dyn_scarfCollar_front_Lower_02_l {parent=dyn_scarfCollar_front_Lower_01_l, loc=<3.5182 0 0>, rot=<0 0 0 1>}
 base_scarfCollar_front_01_l {parent=base_scarfCollar_frontMain_01_l, loc=<-3.6765 0 0>, rot=<0 0 0 1>}
 dyn_scarfCollar_front_01_l {parent=base_scarfCollar_front_01_l, loc=<0 0 0>, rot=<0 0 0 1>}
 dyn_scarfCollar_front_02_l {parent=dyn_scarfCollar_front_01_l, loc=<3.5182 0 0>, rot=<0 0 0 1>}
 base_scarfCollar_backMain_01_r {parent=spine_05, loc=<15.6758 -3.3364 8.8366>, rot=<0.2786 -0.2424 0.9063 -0.2054>}
 base_scarfCollar_back_lower_01_r {parent=base_scarfCollar_backMain_01_r, loc=<5.4422 0 0>, rot=<0 0 1 0>}
 dyn_scarfCollar_back_lower_01_r {parent=base_scarfCollar_back_lower_01_r, loc=<0 0 0>, rot=<0 0 0 1>}
 dyn_scarfCollar_back_lower_02_r {parent=dyn_scarfCollar_back_lower_01_r, loc=<3.5182 0 0>, rot=<0 0 0 1>}
 base_scarfCollar_back_01_r {parent=base_scarfCollar_backMain_01_r, loc=<-5.6618 0 0>, rot=<0 0 0 1>}
 dyn_scarfCollar_back_01_r {parent=base_scarfCollar_back_01_r, loc=<0 0 0>, rot=<0 0 0 1>}
 dyn_scarfCollar_back_02_r {parent=dyn_scarfCollar_back_01_r, loc=<3.5182 0 0>, rot=<0 0 0 1>}
 base_scarfCollar_backMain_01 {parent=spine_05, loc=<13.6902 -7.4114 -0.0646>, rot=<0.1831 -0.6703 0.6888 -0.2069>}
 base_scarfCollar_back_01 {parent=base_scarfCollar_backMain_01, loc=<-6.5272 0 0>, rot=<0 0 0 1>}
 dyn_scarfCollar_back_01 {parent=base_scarfCollar_back_01, loc=<0 0 0>, rot=<0 0 0 1>}
 dyn_scarfCollar_back_02 {parent=dyn_scarfCollar_back_01, loc=<3.5182 0 0>, rot=<0 0 0 1>}
 base_scarfCollar_back_Lower_01 {parent=base_scarfCollar_backMain_01, loc=<5.0824 0 0>, rot=<0 0 1 0>}
 dyn_scarfCollar_back_Lower_01 {parent=base_scarfCollar_back_Lower_01, loc=<0 0 0>, rot=<0 0 0 1>}
 dyn_scarfCollar_back_Lower_02 {parent=dyn_scarfCollar_back_Lower_01, loc=<3.5182 0 0>, rot=<0 0 0 1>}
 base_scarfCollar_backMain_01_l {parent=spine_05, loc=<14.1605 -5.1246 -8.18>, rot=<0.2269 -0.7728 0.5433 -0.2369>}
 base_scarfCollar_back_01_l {parent=base_scarfCollar_backMain_01_l, loc=<-5.9073 0 0>, rot=<0.2358 0 0 -0.9718>}
 dyn_scarfCollar_back_01_l {parent=base_scarfCollar_back_01_l, loc=<0 0 0>, rot=<0 0 0 1>}
 dyn_scarfCollar_back_02_l {parent=dyn_scarfCollar_back_01_l, loc=<3.5182 0 0>, rot=<0 0 0 1>}
 base_scarfCollar_back_lower_01_l {parent=base_scarfCollar_backMain_01_l, loc=<7.4842 0 0>, rot=<0 -0.0145 0.9999 0>}
 dyn_scarfCollar_back_lower_01_l {parent=base_scarfCollar_back_lower_01_l, loc=<0 0 0>, rot=<0 0 0 1>}
 dyn_scarfCollar_back_lower_02_l {parent=dyn_scarfCollar_back_lower_01_l, loc=<3.5182 0 0>, rot=<0 0 0 1>}
 base_BackScarf_01_r {parent=spine_05, loc=<13.0143 -7.2812 8.0689>, rot=<0.3256 -0.5693 0.7496 -0.0889>}
 dyn_BackScarf_01_r {parent=base_BackScarf_01_r, loc=<0 0 0>, rot=<0 0 0 1>}
 dyn_BackScarf_02_r {parent=dyn_BackScarf_01_r, loc=<7.6364 0 0>, rot=<0 -0.0845 0 -0.9964>}
 dyn_BackScarf_03_r {parent=dyn_BackScarf_02_r, loc=<7.636 0 0>, rot=<0 -0.1441 0 -0.9896>}
 dyn_BackScarf_04_r {parent=dyn_BackScarf_03_r, loc=<6.6062 0 0>, rot=<0 0 0 1>}
 head_null_sp {parent=spine_05, type=Null, loc=<20.6573 3.5489 0>, rot=<0 0 -0.036 -0.9994>}
 neck_01_null {parent=spine_05, type=Null, loc=<11.3138 0.1529 0>, rot=<0 0 -0.2003 -0.9797>}
 deform_pec_aim_base_l {parent=spine_04, loc=<7.8732 12.6388 -2.4123>, rot=<0.0102 -0.4602 0.3979 -0.7936>}
 deform_pec_aim_l {parent=deform_pec_aim_base_l, loc=<0 0 0>, rot=<0 0 0 1>}
 deform_pec_base_l {parent=deform_pec_aim_l, loc=<10.766 0 0>, rot=<0 0 0 1>}
 deform_pec_l {parent=deform_pec_base_l, loc=<0.055 0 0>, rot=<0 0 0 1>}
 deform_pec_aim_base_r {parent=spine_04, loc=<7.8727 12.6388 2.4123>, rot=<-0.4602 -0.0102 0.7936 0.3979>}
 deform_pec_aim_r {parent=deform_pec_aim_base_r, loc=<0 0 0>, rot=<0 0 0 1>}
 deform_pec_base_r {parent=deform_pec_aim_r, loc=<-10.7662 -0.0006 -0.0007>, rot=<0 0 0 1>}
 deform_pec_r {parent=deform_pec_base_r, loc=<0 0 0>, rot=<0 0 0 1>}
 thigh_l {parent=pelvis, loc=<-2.5693 -0.0109 -7.7501>, rot=<0.0012 0.0572 0.0155 -0.9982>}
 calf_l {parent=thigh_l, loc=<-39.0484 0 0>, rot=<0.0006 -0.0161 -0.0391 -0.9991>}
 foot_l {parent=calf_l, loc=<-42.561 0.1077 0.4292>, rot=<0.0418 -0.0455 0.0343 -0.9975>}
 ball_l {parent=foot_l, loc=<-6.1599 13.8264 0.6806>, rot=<0.0001 -0.0001 -0.7063 -0.7079>}
 calf_twist_02_l {parent=calf_l, loc=<-18.8064 -0.0967 0.0397>, rot=<0.0237 0.0095 -0.0035 -0.9997>}
 calf_twist_01_l {parent=calf_l, loc=<-35.813 -0.0967 0.0397>, rot=<0.0237 0.0095 -0.0035 -0.9997>}
 deform_knee_in_base_l {parent=calf_l, loc=<1.298 -5.2957 -0.0496>, rot=<0 0 0 -1>}
 deform_knee_in_l {parent=deform_knee_in_base_l, loc=<0 0 0>, rot=<0 0 0 1>}
 deform_knee_out_base_l {parent=calf_l, loc=<0 4 0>, rot=<0 0 0 -1>}
 deform_knee_out_l {parent=deform_knee_out_base_l, loc=<0 0 0>, rot=<0 0 0 1>}
 thigh_twist_01_l {parent=thigh_l, loc=<-21.6303 0 0>, rot=<0.0013 0 0 -1>}
 deform_groin_base_l {parent=thigh_l, loc=<-2.1908 8.1142 -1.7442>, rot=<0 0 0 -1>}
 deform_groin_l {parent=deform_groin_base_l, loc=<0 0 0>, rot=<0 0 0 1>}
 thigh_r {parent=pelvis, loc=<-2.5693 -0.0109 7.7501>, rot=<0.0572 -0.0012 0.9982 0.0155>}
 calf_r {parent=thigh_r, loc=<39.0484 0 0>, rot=<0.0006 -0.0161 -0.0391 -0.9991>}
 foot_r {parent=calf_r, loc=<42.561 -0.1077 -0.4292>, rot=<0.0418 -0.0455 0.0343 -0.9975>}
 ball_r {parent=foot_r, loc=<6.1599 -13.8264 -0.6806>, rot=<0.0001 -0.0001 -0.7063 -0.7079>}
 calf_twist_02_r {parent=calf_r, loc=<18.8064 0.0967 -0.0397>, rot=<0.0237 0.0095 -0.0035 -0.9997>}
 calf_twist_01_r {parent=calf_r, loc=<35.813 0.0967 -0.0397>, rot=<0.0237 0.0095 -0.0035 -0.9997>}
 deform_knee_in_base_r {parent=calf_r, loc=<-1.298 5.2957 0.0496>, rot=<0 0 0 -1>}
 deform_knee_in_r {parent=deform_knee_in_base_r, loc=<0 0 0>, rot=<0 0 0 1>}
 deform_knee_out_base_r {parent=calf_r, loc=<0 -4 0>, rot=<0 0 0 -1>}
 deform_knee_out_r {parent=deform_knee_out_base_r, loc=<0 0 0>, rot=<0 0 0 1>}
 thigh_twist_01_r {parent=thigh_r, loc=<21.6303 0 0>, rot=<0.0013 0 0 -1>}
 deform_groin_base_r {parent=thigh_r, loc=<2.1908 -8.1142 1.7442>, rot=<0 0 0 -1>}
 deform_groin_r {parent=deform_groin_base_r, loc=<0 0 0>, rot=<0 0 0 1>}
 deform_glute_base_l {parent=pelvis, loc=<-8.8064 -8.8222 -8.4287>, rot=<0.0275 0.0501 -0.4472 -0.8926>}
 deform_glute_l {parent=deform_glute_base_l, loc=<0 0 0>, rot=<0 0 0 1>}
 deform_glute_base_r {parent=pelvis, loc=<-8.8062 -8.8222 8.4287>, rot=<0.0501 -0.0275 0.8926 -0.4472>}
 deform_glute_r {parent=deform_glute_base_r, loc=<0 0 0>, rot=<0 0 0 1>}
 attach {parent=root, loc=<0 0 0>, rot=<0 0 0 1>}
 ik_hand_root {parent=root, loc=<0 0 0>, rot=<0 0 0 1>}
 ik_hand_gun {parent=ik_hand_root, loc=<-51.9727 4.0883 102.6935>, rot=<0.6084 0.1793 0.4613 -0.6205>}
 ik_hand_r {parent=ik_hand_gun, loc=<0 0 0>, rot=<0 0 0 1>}
 ik_hand_l {parent=ik_hand_gun, loc=<53.0285 82.1781 35.2051>, rot=<-0.5102 -0.7906 -0.3387 0>}
 ik_foot_root {parent=root, loc=<0 0 0>, rot=<0 0 0 1>}
 ik_foot_l {parent=ik_foot_root, loc=<15.2608 -4.4176 9.4432>, rot=<0.0357 0.7045 0.0284 -0.7082>}
 ik_foot_r {parent=ik_foot_root, loc=<-15.2608 -4.4176 9.4432>, rot=<0.7082 0.0284 -0.7045 0.0357>}
 root_fp {parent=root, loc=<0 0 0>, rot=<0 0 0 1>}
 attach_fp {parent=root_fp, loc=<0 0 0>, rot=<0 0 0 1>}
 camera_root {parent=root_fp, loc=<0 0 0>, rot=<0 0 0 1>}
 camera {parent=camera_root, loc=<0 0 0>, rot=<0 0 0 1>}
 attach_emote_01 {parent=root, loc=<0 0 0>, rot=<0 0 0 1>}
 attach_emote_02 {parent=root, loc=<0 0 0>, rot=<0 0 0 1>}
 attach_emote_03 {parent=root, loc=<0 0 0>, rot=<0 0 0 1>}
 warp_point_01 {parent=root, loc=<0 0 0>, rot=<0 0 0 1>}
 warp_point_02 {parent=root, loc=<0 0 0>, rot=<0 0 0 1>}
 warp_point_03 {parent=root, loc=<0 0 0>, rot=<0 0 0 1>}
]
variables [
 IsTargeting {type=double, default=0}
 SkyDiveFloat {type=double, default=0}
 isInFrontEnd {type=double, default=0}
 is_wear_backpack_float {type=double, default=0}
]
