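# Physics asset for the emote mesh of the M_MED_FrozenTouch male player body, covering its scarf and nearby upper-body bones.
# Capsule bodies form two three-bone scarf chains, back-right and front-left, both rooted on spine_05.
# Scarf joints have zero twist limit, narrow swing1 and swing2 limits that widen toward each tip.
# Upper spine, both upper arms and left forearm add capsules, linked by free angular constraints.
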PhysicsAsset /BRCosmetics/Characters/Player/Male/Medium/Bodies/M_MED_FrozenTouch/Meshes/M_MED_FrozenTouch_Emote_Physics
bodies {
 dyn_BackScarf_01_r [{shapeType=Capsule, length=2.0479, radius=6.1832, shapeName=dyn_BackScarf_01_r_capsule, center=<6.9 0.41 3.67>, rotation=<rot 2.4 -173 -97>}]
 dyn_BackScarf_02_r [{shapeType=Capsule, length=3.661, radius=5.7086, shapeName=dyn_BackScarf_02_r_capsule, center=<7 0.13 3.9>, rotation=<rot -1.5 -175 -90>}]
 dyn_BackScarf_03_r [{shapeType=Capsule, length=4.454, radius=5.8009, shapeName=dyn_BackScarf_03_r_capsule, center=<7 0.65 4>, rotation=<rot 9 -173 -97.2>}]
 dyn_FrontScarf_01_l [{shapeType=Capsule, length=3.7078, radius=2.7293, shapeName=dyn_FrontScarf_01_l_capsule, center=<3.1 0.5 3>, rotation=<rot -7.2 -8.91 -93>}]
 dyn_FrontScarf_02_l [{shapeType=Capsule, length=5.1697, radius=2.6426, shapeName=dyn_FrontScarf_02_l_capsule, center=<4 -1 2.2>, rotation=<rot 8.1 169 -90>}]
 dyn_FrontScarf_03_l [{shapeType=Capsule, length=3.3099, radius=3.7875, shapeName=dyn_FrontScarf_03_l_capsule, center=<7.2 -1.3 1.8>, rotation=<rot -7 0.64 -90>}]
 lowerarm_l [{shapeType=Capsule, length=18.7362, radius=3.9675, shapeName=lowerarm_l_capsule, center=<5.1626 0.864 -0.2097>, rotation=<rot -90 0 0>}]
 spine_04 [
  {shapeType=Capsule, length=15.2112, radius=10.5085, shapeName=spine_04_capsule, center=<-0.2963 -1.1293 -0.3015>}
  {shapeType=Capsule, length=19.3168, radius=10.6832, shapeName=spine_04_capsule, center=<-1.1785 1.6975 0>}
 ]
 spine_05 [
  {shapeType=Capsule, length=19.4847, radius=10.6335, shapeName=spine_05_capsule, center=<0.1376 0.6464 0.2641>}
  {shapeType=Capsule, length=18.7259, radius=11.2741, shapeName=spine_05_capsule, center=<-5.0257 0.4403 0>}
 ]
 upperarm_l [{
   shapeType=Capsule
   length=18.5451
   radius=6.2028
   shapeName=upperarm_l_capsule
   center=<8.093 -3.0648 2.1768>
   rotation=<rot -85.1455 0.0018 179.9983>
  }]
 upperarm_r [{
   shapeType=Capsule
   length=22.7265
   radius=8.526
   shapeName=upperarm_r_capsule
   center=<-12.8635 0.5107 0.9089>
   rotation=<rot -89.0222 179.9957 -176.5407>
  }]
}
constraints [
 {
  bone1Name=spine_05
  bone2Name=spine_04
  swing1LimitDegrees=45
  swing1Motion=Free
  swing2LimitDegrees=45
  swing2Motion=Free
  twistLimitDegrees=45
  twistMotion=Free
 }
 {bone1Name=dyn_BackScarf_01_r, bone2Name=spine_05, swing1LimitDegrees=5, swing2LimitDegrees=15, twistLimitDegrees=0}
 {bone1Name=dyn_BackScarf_02_r, bone2Name=dyn_BackScarf_01_r, swing1LimitDegrees=5, swing2LimitDegrees=20, twistLimitDegrees=0}
 {bone1Name=dyn_BackScarf_03_r, bone2Name=dyn_BackScarf_02_r, swing1LimitDegrees=5, swing2LimitDegrees=25, twistLimitDegrees=0}
 {
  bone1Name=upperarm_r
  bone2Name=spine_05
  swing1LimitDegrees=45
  swing1Motion=Free
  swing2LimitDegrees=45
  swing2Motion=Free
  twistLimitDegrees=45
  twistMotion=Free
 }
 {
  bone1Name=upperarm_l
  bone2Name=spine_05
  swing1LimitDegrees=45
  swing1Motion=Free
  swing2LimitDegrees=45
  swing2Motion=Free
  twistLimitDegrees=45
  twistMotion=Free
 }
 {bone1Name=dyn_FrontScarf_01_l, bone2Name=spine_05, swing1LimitDegrees=5, swing2LimitDegrees=20, twistLimitDegrees=0}
 {bone1Name=dyn_FrontScarf_03_l, bone2Name=dyn_FrontScarf_02_l, swing1LimitDegrees=5, swing2LimitDegrees=30, twistLimitDegrees=0}
 {bone1Name=dyn_FrontScarf_02_l, bone2Name=dyn_FrontScarf_01_l, swing1LimitDegrees=5, swing2LimitDegrees=25, twistLimitDegrees=0}
 {
  bone1Name=lowerarm_l
  bone2Name=upperarm_l
  swing1LimitDegrees=45
  swing1Motion=Free
  swing2LimitDegrees=45
  swing2Motion=Free
  twistLimitDegrees=45
  twistMotion=Free
 }
]
