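# Front-end physics asset for the male medium FrozenTouch body, giving its scarf bones capsule bodies for secondary motion.
# Two three-bone chains, right back scarf and left front scarf, hang from spine_05 with zero twist limits and swing2 limits that widen toward the tips.
# Capsules also cover the upper spine and both arms, linked by free-swing, free-twist constraints; head, pelvis and legs have no bodies.
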PhysicsAsset /BRCosmetics/Characters/Player/Male/Medium/Bodies/M_MED_FrozenTouch/Meshes/M_MED_FrozenTouch_Physics_FrontEnd
bodies {
 dyn_BackScarf_01_r [{
   shapeType=Capsule
   length=2.0479
   radius=6.1832
   shapeName=dyn_BackScarf_01_r_capsule
   center=<6.9016 0.4178 4.5>
   rotation=<rot 2.4 -173 -97>
  }]
 dyn_BackScarf_02_r [{shapeType=Capsule, length=3.661, radius=5.7086, shapeName=dyn_BackScarf_02_r_capsule, center=<6.9984 0 2.8>, rotation=<rot 0 -175 -90>}]
 dyn_BackScarf_03_r [{
   shapeType=Capsule
   length=4.454
   radius=5.8009
   shapeName=dyn_BackScarf_03_r_capsule
   center=<7.0086 0.6289 3.9606>
   rotation=<rot 9 -173 -97.2>
  }]
 dyn_FrontScarf_01_l [{shapeType=Capsule, length=7.8397, radius=3.3676, shapeName=dyn_FrontScarf_01_l_capsule, center=<3.2 -0.8 3.2811>, rotation=<rot 0 0 -90>}]
 dyn_FrontScarf_02_l [{shapeType=Capsule, length=7.1072, radius=3.4897, shapeName=dyn_FrontScarf_02_l_capsule, center=<4.3 -1 3.2129>, rotation=<rot 0 0 -90>}]
 dyn_FrontScarf_03_l [{
   shapeType=Capsule
   length=9.0024
   radius=4.5062
   shapeName=dyn_FrontScarf_03_l_capsule
   center=<7.3 -1.3966 2.4112>
   rotation=<rot 0.0001 0 -84.2123>
  }]
 lowerarm_l [{
   shapeType=Capsule
   length=31.6474
   radius=5.1091
   shapeName=lowerarm_l_capsule
   center=<15.466 -0.9583 -0.47>
   rotation=<rot 83.4164 1.1197 3.3352>
  }]
 lowerarm_r [{
   shapeType=Capsule
   length=31.6474
   radius=5.1091
   shapeName=lowerarm_r_capsule
   center=<-15.466 0.9583 0.47>
   rotation=<rot 83.4164 1.1197 -176.6648>
  }]
 spine_04 [
  {shapeType=Capsule, length=15.2112, radius=10.5085, shapeName=spine_04_capsule, center=<-0.2963 -1.1293 -0.3015>}
  {shapeType=Capsule, length=19.3168, radius=10.6832, shapeName=spine_04_capsule, center=<-1.1785 1.233 0>}
 ]
 spine_05 [
  {shapeType=Capsule, length=19.4847, radius=10.6335, shapeName=spine_05_capsule, center=<0.9954 0.123 0.2641>}
  {shapeType=Capsule, length=18.7259, radius=11.2741, shapeName=spine_05_capsule, center=<-5.0257 -0.172 0>}
 ]
 upperarm_l [{
   shapeType=Capsule
   length=17.6783
   radius=7.0696
   shapeName=upperarm_l_capsule
   center=<8.3774 -1.503 -1.1725>
   rotation=<rot -85.1455 0.0018 179.9983>
  }]
 upperarm_r [{
   shapeType=Capsule
   length=21.4123
   radius=8.526
   shapeName=upperarm_r_capsule
   center=<-13.1085 0.4959 0.9131>
   rotation=<rot -89.0222 179.9957 -176.5407>
  }]
}
constraints [
 {
  bone1Name=spine_05
  bone2Name=spine_04
  swing1LimitDegrees=45
  swing1Motion=Free
  swing2LimitDegrees=45
  swing2Motion=Free
  twistLimitDegrees=45
  twistMotion=Free
 }
 {bone1Name=dyn_BackScarf_01_r, bone2Name=spine_05, swing1LimitDegrees=5, swing2LimitDegrees=10, twistLimitDegrees=0}
 {bone1Name=dyn_BackScarf_02_r, bone2Name=dyn_BackScarf_01_r, swing1LimitDegrees=5, swing2LimitDegrees=12, twistLimitDegrees=0}
 {bone1Name=dyn_BackScarf_03_r, bone2Name=dyn_BackScarf_02_r, swing1LimitDegrees=5, swing2LimitDegrees=15, twistLimitDegrees=0}
 {
  bone1Name=upperarm_r
  bone2Name=spine_05
  swing1LimitDegrees=45
  swing1Motion=Free
  swing2LimitDegrees=45
  swing2Motion=Free
  twistLimitDegrees=45
  twistMotion=Free
 }
 {
  bone1Name=upperarm_l
  bone2Name=spine_05
  swing1LimitDegrees=45
  swing1Motion=Free
  swing2LimitDegrees=45
  swing2Motion=Free
  twistLimitDegrees=45
  twistMotion=Free
 }
 {bone1Name=dyn_FrontScarf_01_l, bone2Name=spine_05, swing1LimitDegrees=5, swing2LimitDegrees=20, twistLimitDegrees=0}
 {bone1Name=dyn_FrontScarf_03_l, bone2Name=dyn_FrontScarf_02_l, swing1LimitDegrees=5, swing2LimitDegrees=30, twistLimitDegrees=0}
 {bone1Name=dyn_FrontScarf_02_l, bone2Name=dyn_FrontScarf_01_l, swing1LimitDegrees=5, swing2LimitDegrees=25, twistLimitDegrees=0}
 {
  bone1Name=lowerarm_l
  bone2Name=upperarm_l
  swing1LimitDegrees=45
  swing1Motion=Free
  swing2LimitDegrees=45
  swing2Motion=Free
  twistLimitDegrees=45
  twistMotion=Free
 }
 {
  bone1Name=lowerarm_r
  bone2Name=upperarm_r
  swing1LimitDegrees=45
  swing1Motion=Free
  swing2LimitDegrees=45
  swing2Motion=Free
  twistLimitDegrees=45
  twistMotion=Free
 }
]
