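# Post-process-style AnimBlueprint for the FrozenTouch male cosmetic body, copying the parent mesh pose and layering secondary motion.
# Three spine_05-space RigidBody nodes use separate physics assets: default (no backbling, emote or front end), front end, and emote without backbling.
# Emote without backbling also offsets two left front-scarf bones before a final Control Rig receives backbling, targeting, skydive and front-end floats.
# The EventGraph updates these flags and stores parameter-asset gravity settings.
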AnimBlueprint /BRCosmetics/Characters/Player/Male/Medium/Bodies/M_MED_FrozenTouch/Meshes/M_MED_FrozenTouch_AnimBP
target_skeleton=&/M_MED_FrozenTouch_Skeleton
graphs [
 AnimGraph {nodes [
   n1 = RigidBody!(
    ComponentPose=Local To Component(LocalPose=Copy Pose From Mesh() {bUseAttachedParent=true})
    bAlphaBoolEnabled:bool=NOT Boolean(A=Get isWearingPackAndPlayingEmoteAndInFrontEnd() -> (isWearingPackAndPlayingEmoteAndInFrontEnd:bool))
    OverrideWorldGravity:Vector="0,0,0"
    ExternalForce
    SimSpaceSettings:Sim Space Settings Structure=(WorldAlpha=0,VelocityScaleZ=1,DampingAlpha=1,MaxLinearVelocity=10000,MaxAngularVelocity=10000,MaxLinearAcceleration=10000,MaxAngularAcceleration=10000,ExternalLinearDragV=(X=0,Y=0,Z=0),ExternalLinearVelocity=(X=0,Y=0,Z=0),ExternalAngularVelocity=(X=0,Y=0,Z=0))
   ) -> (Pose) {
    alphaInputType=Bool
    bOverrideWorldGravity=true
    overridePhysicsAsset=&/M_MED_FrozenTouch_Physics
    simulationSpace=BaseBoneSpace
    baseBoneRef {boneName=spine_05}
   }
   n2 = RigidBody(
    ComponentPose=@n1
    bAlphaBoolEnabled=Get bIsInFrontEnd()
    OverrideWorldGravity="0,0,0"
    SimSpaceSettings=(WorldAlpha=0,VelocityScaleZ=1,DampingAlpha=1,MaxLinearVelocity=10000,MaxAngularVelocity=10000,MaxLinearAcceleration=10000,MaxAngularAcceleration=10000,ExternalLinearDragV=(X=0,Y=0,Z=0),ExternalLinearVelocity=(X=0,Y=0,Z=0),ExternalAngularVelocity=(X=0,Y=0,Z=0))
   ) {
    alphaInputType=Bool
    bOverrideWorldGravity=true
    overridePhysicsAsset=&/M_MED_FrozenTouch_Physics_FrontEnd
    simulationSpace=BaseBoneSpace
    baseBoneRef {boneName=spine_05}
   }
   n3 = Get isWearingPackAndPlayingEmote_Not() -> (isWearingPackAndPlayingEmote_Not:bool)
   n4 = RigidBody(
    ComponentPose=@n2
    bAlphaBoolEnabled=@n3
    OverrideWorldGravity="0,0,0"
    SimSpaceSettings=(WorldAlpha=0,VelocityScaleZ=1,DampingAlpha=1,MaxLinearVelocity=10000,MaxAngularVelocity=10000,MaxLinearAcceleration=10000,MaxAngularAcceleration=10000,ExternalLinearDragV=(X=0,Y=0,Z=0),ExternalLinearVelocity=(X=0,Y=0,Z=0),ExternalAngularVelocity=(X=0,Y=0,Z=0))
   ) {
    alphaInputType=Bool
    bOverrideWorldGravity=true
    overridePhysicsAsset=&/M_MED_FrozenTouch_Emote_Physics
    simulationSpace=BaseBoneSpace
    baseBoneRef {boneName=spine_05}
   }
   n5 = "Transform (Modify) Bone - Bone: base_FrontScarf_01_l"!(ComponentPose=@n4, bAlphaBoolEnabled:bool=@n3, Translation="0,-2,0", Rotation="0,10,0", Scale="1,1,1") -> (Pose) {
    alphaInputType=Bool
    rotationMode=BMM_Additive
    rotationSpace=BCS_BoneSpace
    translationMode=BMM_Additive
    translationSpace=BCS_BoneSpace
    alphaBoolBlend {blendInTime=10, blendOutTime=10}
    boneToModify {boneName=base_FrontScarf_01_l}
    rotation {yaw=10}
    translation {y=-2}
   }
   n6 = "Transform (Modify) Bone - Bone: base_FrontScarf_02_l"!(ComponentPose=@n5, bAlphaBoolEnabled:bool=@n3, Translation="0,0,0", Rotation="5,0,0", Scale="1,1,1") -> (Pose) {
    alphaInputType=Bool
    rotationMode=BMM_Additive
    rotationSpace=BCS_BoneSpace
    translationMode=BMM_Additive
    translationSpace=BCS_BoneSpace
    alphaBoolBlend {blendInTime=10, blendOutTime=10}
    boneToModify {boneName=base_FrontScarf_02_l}
    rotation {pitch=5}
   }
   n7 = Control Rig!(
    Source=Component To Local(ComponentPose=@n6)
    Alpha
    is_wear_backpack_float:double=Get isWearingBackblingFloat() -> (isWearingBackblingFloat:double)
    IsTargeting:double=Get isTargetingFloat() -> (isTargetingFloat:double)
    SkyDiveFloat:double=Get skyDiveFloat() -> (skyDiveFloat:double)
    isInFrontEnd:double=Get isInFrontEndFloat() -> (isInFrontEndFloat:double)
   ) -> (Pose) {controlRigAssetReference {blueprintRigClass=&/M_MED_FrozenTouch_CtrlRig1.M_MED_FrozenTouch_CtrlRig1_C}}
   Output Pose(Result=@n7) {layerGroup=None}
  ]}
 EventGraph {nodes [
   Event BlueprintUpdateAnimation()
   > n1 = Sequence!(execute) -> (then_0, then_1, then_2:Exec)
   n2 = Set isWearingBackbling(execute=@n1.then_1, isWearingBackbling=CheckForBackbling())
   > Set isWearingBackblingFloat(isWearingBackblingFloat:double="To Float (Boolean)"(InBool=@n2)) -> (Output_Get:double)
   Set isWearingPackAndPlayingEmoteAndInFrontEnd(
    execute=@n1.then_2
    isWearingPackAndPlayingEmoteAndInFrontEnd:bool=OR Boolean!(A=Get isWearingBackbling(), B=Get bIsPlayingEmote(), C:bool=Get bIsInFrontEnd()) -> (ReturnValue)
   ) -> (Output_Get:bool)
   n3 = Get isWearingBackbling()
   > Set isWearingPackAndPlayingEmote_Not(isWearingPackAndPlayingEmote_Not:bool=AND Boolean(A=NOT Boolean(A=@n3), B=Get bIsPlayingEmote())) -> (Output_Get:bool)
   > Set isTargetingFloat(isTargetingFloat:double="To Float (Boolean)"(InBool=Get bIsTargeting())) -> (Output_Get:double)
   > Set skyDiveFloat(skyDiveFloat:double="To Float (Boolean)"(InBool=Get bIsSkydiveFloating())) -> (Output_Get:double)
   > Set isInFrontEndFloat(isInFrontEndFloat:double="To Float (Boolean)"(InBool=Get bIsInFrontEnd())) -> (Output_Get:double)
   n4 = "Gravity Override State Function (Parameter Asset) (Nativized)"(
    execute=@n1.then_0
    CharacterPartAnimInstance=Self-Reference()
    GameTimeInSeconds=GetGameTimeInSeconds()
    Parameters=&/F_MED_FrozenTouch_Parameters
   )
   > Set GravityOverride(GravityOverride=@n4)
   > Set RBAN Settings Out(RBAN Settings Out=@n4.OutRigidBodyAnimNodeParameters)
  ]}
]
variables [
 GravityOverride {type=Vector}
 isInFrontEndFloat {type=number}
 isTargetingFloat {type=number}
 isWearingBackbling {type=boolean}
 isWearingBackblingFloat {type=number}
 isWearingPackAndPlayingEmote_Not {type=boolean}
 isWearingPackAndPlayingEmoteAndInFrontEnd {type=boolean}
 RBAN Settings Out {type=FortRigidBodyAnimNodeParameters}
 skyDiveFloat {type=number}
]
